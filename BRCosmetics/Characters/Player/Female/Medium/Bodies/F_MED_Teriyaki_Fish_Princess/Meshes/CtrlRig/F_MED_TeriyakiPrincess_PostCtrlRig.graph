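# Post-process Control Rig for the Teriyaki Fish Princess body that drives skirt bones with offset-maintaining transform constraints.
# Each side-front skirt bone follows its matching front and side-back bones.
# Front root bones blend between pelvis and same-side thigh, with the clamped `crouching` variable shifting weight fully onto the thigh.
# Only the left side-front chain connects to Begin Execution; the front-root and right side-front chains are unconnected.
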ControlRigBlueprint /BRCosmetics/Characters/Player/Female/Medium/Bodies/F_MED_Teriyaki_Fish_Princess/Meshes/CtrlRig/F_MED_TeriyakiPrincess_PostCtrlRig
graphs [
 RigVMFunctionLibrary {links [], nodes []}
 RigVMModel {nodes [
   RigUnit_BeginExecution = &RigUnit_BeginExecution()
   /Script/ControlRig.RigUnit_TransformConstraint(
    BaseTransform=(Rotation=(X=0,Y=0,Z=0,W=1),Translation=(X=0,Y=0,Z=0),Scale3D=(X=1,Y=1,Z=1))
    BaseTransformSpace=GlobalSpace
    Bone=dyn_skirtSideFront_01_l
    Targets="((Transform=(Rotation=(X=0,Y=0,Z=0,W=1),Translation=(X=0,Y=0,Z=0),Scale3D=(X=1,Y=1,Z=1)),Weight=1,bMaintainOffset=True,Filter=(TranslationFilter=(bX=True,bY=True,bZ=True),RotationFilter=(bX=True,bY=True,bZ=True),ScaleFilter=(bX=True,bY=True,bZ=True))),(Tra ...(truncated)"
    bUseInitialTransforms=True
    Targets.0.Transform=&RigUnit_GetBoneTransform(Bone=dyn_skirtFront_01_l).Transform
    Targets.1.Transform=&RigUnit_GetBoneTransform(Bone=dyn_skirtSideBack_01_l).Transform
    ExecutePin=@RigUnit_BeginExecution.ExecutePin
   ) {is=struct RigVMUnitNode}
   > &RigUnit_TransformConstraint(
    BaseTransform=(Rotation=(X=0,Y=0,Z=0,W=1),Translation=(X=0,Y=0,Z=0),Scale3D=(X=1,Y=1,Z=1))
    BaseTransformSpace=GlobalSpace
    Bone=dyn_skirtSideFront_02_l
    Targets="((Transform=(Rotation=(X=0,Y=0,Z=0,W=1),Translation=(X=0,Y=0,Z=0),Scale3D=(X=1,Y=1,Z=1)),Weight=1,bMaintainOffset=True,Filter=(TranslationFilter=(bX=True,bY=True,bZ=True),RotationFilter=(bX=True,bY=True,bZ=True),ScaleFilter=(bX=True,bY=True,bZ=True))),(Tra ...(truncated)"
    bUseInitialTransforms=True
    Targets.1.Transform=&RigUnit_GetBoneTransform(Bone=dyn_skirtSideBack_02_l).Transform
    Targets.0.Transform=&RigUnit_GetBoneTransform(Bone=dyn_skirtFront_02_l).Transform
   )
   > &RigUnit_TransformConstraint(
    BaseTransform=(Rotation=(X=0,Y=0,Z=0,W=1),Translation=(X=0,Y=0,Z=0),Scale3D=(X=1,Y=1,Z=1))
    BaseTransformSpace=GlobalSpace
    Bone=dyn_skirtSideFront_03_l
    Targets="((Transform=(Rotation=(X=0,Y=0,Z=0,W=1),Translation=(X=0,Y=0,Z=0),Scale3D=(X=1,Y=1,Z=1)),Weight=1,bMaintainOffset=True,Filter=(TranslationFilter=(bX=True,bY=True,bZ=True),RotationFilter=(bX=True,bY=True,bZ=True),ScaleFilter=(bX=True,bY=True,bZ=True))),(Tra ...(truncated)"
    bUseInitialTransforms=True
    Targets.1.Transform=&RigUnit_GetBoneTransform(Bone=dyn_skirtSideBack_03_l).Transform
    Targets.0.Transform=&RigUnit_GetBoneTransform(Bone=dyn_skirtFront_03_l).Transform
   )
   > &RigUnit_TransformConstraint(
    BaseTransform=(Rotation=(X=0,Y=0,Z=0,W=1),Translation=(X=0,Y=0,Z=0),Scale3D=(X=1,Y=1,Z=1))
    BaseTransformSpace=GlobalSpace
    Bone=dyn_skirtSideFront_04_l
    Targets="((Transform=(Rotation=(X=0,Y=0,Z=0,W=1),Translation=(X=0,Y=0,Z=0),Scale3D=(X=1,Y=1,Z=1)),Weight=1,bMaintainOffset=True,Filter=(TranslationFilter=(bX=True,bY=True,bZ=True),RotationFilter=(bX=True,bY=True,bZ=True),ScaleFilter=(bX=True,bY=True,bZ=True))),(Tra ...(truncated)"
    bUseInitialTransforms=True
    Targets.1.Transform=&RigUnit_GetBoneTransform(Bone=dyn_skirtSideBack_04_l).Transform
    Targets.0.Transform=&RigUnit_GetBoneTransform(Bone=dyn_skirtFront_04_l).Transform
   )
   > &RigUnit_TransformConstraint(
    BaseTransform=(Rotation=(X=0,Y=0,Z=0,W=1),Translation=(X=0,Y=0,Z=0),Scale3D=(X=1,Y=1,Z=1))
    BaseTransformSpace=GlobalSpace
    Bone=dyn_skirtSideFront_05_l
    Targets="((Transform=(Rotation=(X=0,Y=0,Z=0,W=1),Translation=(X=0,Y=0,Z=0),Scale3D=(X=1,Y=1,Z=1)),Weight=1,bMaintainOffset=True,Filter=(TranslationFilter=(bX=True,bY=True,bZ=True),RotationFilter=(bX=True,bY=True,bZ=True),ScaleFilter=(bX=True,bY=True,bZ=True))),(Tra ...(truncated)"
    bUseInitialTransforms=True
    Targets.1.Transform=&RigUnit_GetBoneTransform(Bone=dyn_skirtSideBack_05_l).Transform
    Targets.0.Transform=&RigUnit_GetBoneTransform(Bone=dyn_skirtFront_05_l).Transform
   )
   n1 = &RigUnit_GetBoneTransform(Bone=pelvis)
   n2 = RigVMVariableNode() {variable=crouching}
   n3 = &RigVMFunction_MathFloatRemap(SourceMaximum=1, SourceMinimum=0, TargetMaximum=0, TargetMinimum=0.35, bClamp=true, Value=@n2.Value)
   n4 = &RigVMFunction_MathFloatRemap(SourceMaximum=1, SourceMinimum=0, TargetMinimum=0.65, bClamp=true, Value=@n2.Value)
   &RigUnit_TransformConstraint(
    BaseTransform=(Rotation=(X=0,Y=0,Z=0,W=1),Translation=(X=0,Y=0,Z=0),Scale3D=(X=1,Y=1,Z=1))
    BaseTransformSpace=GlobalSpace
    Bone=dyn_skirtFront_01_l
    ExecutePin=()
    Targets="((Transform=(Rotation=(X=0,Y=0,Z=0,W=1),Translation=(X=0,Y=0,Z=0),Scale3D=(X=1,Y=1,Z=1)),Weight=0.2,bMaintainOffset=True,Filter=(TranslationFilter=(bX=True,bY=True,bZ=True),RotationFilter=(bX=True,bY=True,bZ=True),ScaleFilter=(bX=True,bY=True,bZ=True))),(T ...(truncated)"
    bUseInitialTransforms=True
    Targets.0.Transform=@n1.Transform
    Targets.1.Transform=&RigUnit_GetBoneTransform(Bone=thigh_l).Transform
    Targets.0.Weight=@n3.Result
    Targets.1.Weight=@n4.Result
   )
   > &RigUnit_TransformConstraint(
    BaseTransform=(Rotation=(X=0,Y=0,Z=0,W=1),Translation=(X=0,Y=0,Z=0),Scale3D=(X=1,Y=1,Z=1))
    BaseTransformSpace=GlobalSpace
    Bone=dyn_skirtFront_01_r
    Targets="((Transform=(Rotation=(X=0,Y=0,Z=0,W=1),Translation=(X=0,Y=0,Z=0),Scale3D=(X=1,Y=1,Z=1)),Weight=1,bMaintainOffset=True,Filter=(TranslationFilter=(bX=True,bY=True,bZ=True),RotationFilter=(bX=True,bY=True,bZ=True),ScaleFilter=(bX=True,bY=True,bZ=True))),(Tra ...(truncated)"
    bUseInitialTransforms=True
    Targets.0.Transform=@n1.Transform
    Targets.1.Transform=&RigUnit_GetBoneTransform(Bone=thigh_r).Transform
    Targets.0.Weight=@n3.Result
    Targets.1.Weight=@n4.Result
   )
   &RigUnit_TransformConstraint(
    BaseTransform=(Rotation=(X=0,Y=0,Z=0,W=1),Translation=(X=0,Y=0,Z=0),Scale3D=(X=1,Y=1,Z=1))
    BaseTransformSpace=GlobalSpace
    Bone=dyn_skirtSideFront_01_r
    ExecutePin=()
    Targets="((Transform=(Rotation=(X=0,Y=0,Z=0,W=1),Translation=(X=0,Y=0,Z=0),Scale3D=(X=1,Y=1,Z=1)),Weight=1,bMaintainOffset=True,Filter=(TranslationFilter=(bX=True,bY=True,bZ=True),RotationFilter=(bX=True,bY=True,bZ=True),ScaleFilter=(bX=True,bY=True,bZ=True))),(Tra ...(truncated)"
    bUseInitialTransforms=True
    Targets.1.Transform=&RigUnit_GetBoneTransform(Bone=dyn_skirtSideBack_01_r).Transform
    Targets.0.Transform=&RigUnit_GetBoneTransform(Bone=dyn_skirtFront_01_r).Transform
   )
   > &RigUnit_TransformConstraint(
    BaseTransform=(Rotation=(X=0,Y=0,Z=0,W=1),Translation=(X=0,Y=0,Z=0),Scale3D=(X=1,Y=1,Z=1))
    BaseTransformSpace=GlobalSpace
    Bone=dyn_skirtSideFront_02_r
    Targets="((Transform=(Rotation=(X=0,Y=0,Z=0,W=1),Translation=(X=0,Y=0,Z=0),Scale3D=(X=1,Y=1,Z=1)),Weight=1,bMaintainOffset=True,Filter=(TranslationFilter=(bX=True,bY=True,bZ=True),RotationFilter=(bX=True,bY=True,bZ=True),ScaleFilter=(bX=True,bY=True,bZ=True))),(Tra ...(truncated)"
    bUseInitialTransforms=True
    Targets.1.Transform=&RigUnit_GetBoneTransform(Bone=dyn_skirtSideBack_02_r).Transform
    Targets.0.Transform=&RigUnit_GetBoneTransform(Bone=dyn_skirtFront_02_r).Transform
   )
   > &RigUnit_TransformConstraint(
    BaseTransform=(Rotation=(X=0,Y=0,Z=0,W=1),Translation=(X=0,Y=0,Z=0),Scale3D=(X=1,Y=1,Z=1))
    BaseTransformSpace=GlobalSpace
    Bone=dyn_skirtSideFront_03_r
    Targets="((Transform=(Rotation=(X=0,Y=0,Z=0,W=1),Translation=(X=0,Y=0,Z=0),Scale3D=(X=1,Y=1,Z=1)),Weight=1,bMaintainOffset=True,Filter=(TranslationFilter=(bX=True,bY=True,bZ=True),RotationFilter=(bX=True,bY=True,bZ=True),ScaleFilter=(bX=True,bY=True,bZ=True))),(Tra ...(truncated)"
    bUseInitialTransforms=True
    Targets.1.Transform=&RigUnit_GetBoneTransform(Bone=dyn_skirtSideBack_03_r).Transform
    Targets.0.Transform=&RigUnit_GetBoneTransform(Bone=dyn_skirtFront_03_r).Transform
   )
   > &RigUnit_TransformConstraint(
    BaseTransform=(Rotation=(X=0,Y=0,Z=0,W=1),Translation=(X=0,Y=0,Z=0),Scale3D=(X=1,Y=1,Z=1))
    BaseTransformSpace=GlobalSpace
    Bone=dyn_skirtSideFront_04_r
    Targets="((Transform=(Rotation=(X=0,Y=0,Z=0,W=1),Translation=(X=0,Y=0,Z=0),Scale3D=(X=1,Y=1,Z=1)),Weight=1,bMaintainOffset=True,Filter=(TranslationFilter=(bX=True,bY=True,bZ=True),RotationFilter=(bX=True,bY=True,bZ=True),ScaleFilter=(bX=True,bY=True,bZ=True))),(Tra ...(truncated)"
    bUseInitialTransforms=True
    Targets.1.Transform=&RigUnit_GetBoneTransform(Bone=dyn_skirtSideBack_04_r).Transform
    Targets.0.Transform=&RigUnit_GetBoneTransform(Bone=dyn_skirtFront_04_r).Transform
   )
   > &RigUnit_TransformConstraint(
    BaseTransform=(Rotation=(X=0,Y=0,Z=0,W=1),Translation=(X=0,Y=0,Z=0),Scale3D=(X=1,Y=1,Z=1))
    BaseTransformSpace=GlobalSpace
    Bone=dyn_skirtSideFront_05_r
    Targets="((Transform=(Rotation=(X=0,Y=0,Z=0,W=1),Translation=(X=0,Y=0,Z=0),Scale3D=(X=1,Y=1,Z=1)),Weight=1,bMaintainOffset=True,Filter=(TranslationFilter=(bX=True,bY=True,bZ=True),RotationFilter=(bX=True,bY=True,bZ=True),ScaleFilter=(bX=True,bY=True,bZ=True))),(Tra ...(truncated)"
    bUseInitialTransforms=True
    Targets.1.Transform=&RigUnit_GetBoneTransform(Bone=dyn_skirtSideBack_05_r).Transform
    Targets.0.Transform=&RigUnit_GetBoneTransform(Bone=dyn_skirtFront_05_r).Transform
   )
  ]}
]
hierarchy [
 root {loc=<0 0 0>, rot=<0 0 0 1>}
 attach {parent=root, loc=<0 0 0>, rot=<0 0 0 1>}
 pelvis {parent=root, loc=<0 -0.7071 93.1191>, rot=<0.0127 0.707 -0.0127 -0.707>}
 spine_01 {parent=pelvis, loc=<6.7494 -0.6388 0>, rot=<0 0 0.034 -0.9994>}
 spine_02 {parent=spine_01, loc=<7.2518 -0.035 0>, rot=<0 0 0.0238 -0.9997>}
 spine_03 {parent=spine_02, loc=<6.0104 0 0>, rot=<0 0 0.0149 -0.9999>}
 spine_04 {parent=spine_03, loc=<8.1024 0 0>, rot=<0 0 -0.0132 -0.9999>}
 spine_05 {parent=spine_04, loc=<10.0146 0 0>, rot=<0 0 -0.0015 -1>}
 clavicle_l {parent=spine_05, loc=<4.3575 -0.1466 -2.3223>, rot=<0.0465 -0.7289 -0.009 -0.6829>}
 upperarm_l {parent=clavicle_l, loc=<15.5111 0 0>, rot=<0.0226 -0.3234 -0.0088 -0.9459>}
 lowerarm_l {parent=upperarm_l, loc=<26.6514 0 0>, rot=<0.0391 -0.0099 -0.2453 -0.9686>}
 hand_l {parent=lowerarm_l, loc=<22.3491 0 0>, rot=<-0.7081 -0.0507 0.0975 -0.6975>}
 index_metacarpal_l {parent=hand_l, loc=<1.6535 0.0183 -2.96>, rot=<0.0158 -0.0124 -0.0173 -0.9996>}
 index_01_l {parent=index_metacarpal_l, loc=<8.3008 0.0357 -0.0524>, rot=<-0.0044 0.0485 0.0924 -0.9945>}
 index_02_l {parent=index_01_l, loc=<3.3587 -0.3004 0.0452>, rot=<-0.002 0.0027 0.3225 -0.9466>}
 index_03_l {parent=index_02_l, loc=<2.5578 -0.123 -0.0252>, rot=<-0.002 0.0067 0.2558 -0.9667>}
 middle_metacarpal_l {parent=hand_l, loc=<2.1856 -0.1725 -0.5984>, rot=<0.006 0.0168 -0.0329 -0.9993>}
 middle_01_l {parent=middle_metacarpal_l, loc=<7.3275 -0.2447 0.0129>, rot=<-0.0208 0.0042 0.0858 -0.9961>}
 middle_02_l {parent=middle_01_l, loc=<4.0123 0.1401 0.1175>, rot=<-0.0218 0.0258 0.3512 -0.9357>}
 middle_03_l {parent=middle_02_l, loc=<2.6077 0.0534 0.1373>, rot=<0.0082 0.0188 0.1708 -0.9851>}
 pinky_metacarpal_l {parent=hand_l, loc=<1.7616 -0.9753 3.3814>, rot=<-0.0397 0.0814 -0.069 -0.9935>}
 pinky_01_l {parent=pinky_metacarpal_l, loc=<7.1709 -0.4831 -0.1094>, rot=<-0.0654 -0.0035 0.2242 -0.9723>}
 pinky_02_l {parent=pinky_01_l, loc=<2.8274 0.0893 -0.1043>, rot=<0.0037 0.0011 0.2854 -0.9584>}
 pinky_03_l {parent=pinky_02_l, loc=<2.2046 0.0307 -0.0078>, rot=<-0.0002 0.0098 0.1445 -0.9895>}
 ring_metacarpal_l {parent=hand_l, loc=<2.2074 -0.3625 1.3638>, rot=<-0.0284 0.0722 -0.0458 -0.9959>}
 ring_01_l {parent=ring_metacarpal_l, loc=<7.0295 -0.6197 -0.1446>, rot=<-0.0489 -0.0252 0.1892 -0.9804>}
 ring_02_l {parent=ring_01_l, loc=<3.6174 0.4646 -0.1027>, rot=<0 0 0.2806 -0.9598>}
 ring_03_l {parent=ring_02_l, loc=<2.4152 0.1141 0.0199>, rot=<0 0 0.1189 -0.9929>}
 thumb_01_l {parent=hand_l, loc=<1.3118 -0.3191 -2.459>, rot=<0.6546 -0.4042 0.0664 -0.6354>}
 thumb_02_l {parent=thumb_01_l, loc=<5.9336 -0.0448 0.0401>, rot=<-0.0037 0.0213 0.298 -0.9543>}
 thumb_03_l {parent=thumb_02_l, loc=<2.2223 0.3682 -0.0907>, rot=<0.0118 -0.013 0.0066 -0.9998>}
 weapon_l {parent=hand_l, loc=<-0.1974 -3.4111 0.3811>, rot=<0 0 0 1>}
 lowerarm_twist_01_l {parent=lowerarm_l, loc=<17.8549 0 0>, rot=<0 0 0 1>}
 lowerarm_twist_02_l {parent=lowerarm_l, loc=<9.6199 0 0>, rot=<0 0 0 1>}
 upperarm_twist_01_l {parent=upperarm_l, loc=<0.0041 0 0>, rot=<0 0 0 -1>}
 upperarm_twist_02_l {parent=upperarm_l, loc=<13.7941 0 0>, rot=<0 0 0 -1>}
 clavicle_r {parent=spine_05, loc=<4.3575 -0.1466 2.3223>, rot=<0.7289 0.0465 -0.6829 0.009>}
 upperarm_r {parent=clavicle_r, loc=<-15.5111 0 0>, rot=<0.0226 -0.3234 -0.0088 -0.9459>}
 lowerarm_r {parent=upperarm_r, loc=<-26.6514 0 0>, rot=<0.0391 -0.0099 -0.2453 -0.9686>}
 hand_r {parent=lowerarm_r, loc=<-22.3492 0 0>, rot=<-0.7081 -0.0507 0.0975 -0.6975>}
 index_metacarpal_r {parent=hand_r, loc=<-1.6534 -0.0181 2.9601>, rot=<0.0158 -0.0124 -0.0173 -0.9996>}
 index_01_r {parent=index_metacarpal_r, loc=<-8.3008 -0.0357 0.0524>, rot=<-0.0044 0.0485 0.0924 -0.9945>}
 index_02_r {parent=index_01_r, loc=<-3.3589 0.3003 -0.0453>, rot=<-0.002 0.0027 0.3225 -0.9466>}
 index_03_r {parent=index_02_r, loc=<-2.5578 0.1231 0.0251>, rot=<-0.002 0.0067 0.2558 -0.9667>}
 middle_metacarpal_r {parent=hand_r, loc=<-2.1856 0.1726 0.5984>, rot=<0.006 0.0168 -0.0329 -0.9993>}
 middle_01_r {parent=middle_metacarpal_r, loc=<-7.3273 0.2447 -0.0129>, rot=<-0.0208 0.0042 0.0858 -0.9961>}
 middle_02_r {parent=middle_01_r, loc=<-4.0142 -0.1411 -0.07>, rot=<-0.0217 0.0259 0.3512 -0.9357>}
 middle_03_r {parent=middle_02_r, loc=<-2.6074 -0.0534 -0.1312>, rot=<0.0083 0.0188 0.1708 -0.9851>}
 pinky_metacarpal_r {parent=hand_r, loc=<-1.7617 0.9753 -3.3814>, rot=<-0.0397 0.0814 -0.069 -0.9935>}
 pinky_01_r {parent=pinky_metacarpal_r, loc=<-7.1707 0.4832 0.1094>, rot=<-0.0654 -0.0035 0.2242 -0.9723>}
 pinky_02_r {parent=pinky_01_r, loc=<-2.8274 -0.0892 0.1043>, rot=<0.0037 0.0011 0.2854 -0.9584>}
 pinky_03_r {parent=pinky_02_r, loc=<-2.2045 -0.031 0.0078>, rot=<-0.0002 0.0098 0.1445 -0.9895>}
 ring_metacarpal_r {parent=hand_r, loc=<-2.2073 0.3627 -1.3638>, rot=<-0.0284 0.0722 -0.0458 -0.9959>}
 ring_01_r {parent=ring_metacarpal_r, loc=<-7.0297 0.6197 0.1446>, rot=<-0.0489 -0.0252 0.1892 -0.9804>}
 ring_02_r {parent=ring_01_r, loc=<-3.6172 -0.4648 0.1027>, rot=<0 0 0.2806 -0.9598>}
 ring_03_r {parent=ring_02_r, loc=<-2.4152 -0.1141 -0.0199>, rot=<0 0 0.1189 -0.9929>}
 thumb_01_r {parent=hand_r, loc=<-1.3119 0.3192 2.459>, rot=<0.6546 -0.4042 0.0664 -0.6354>}
 thumb_02_r {parent=thumb_01_r, loc=<-5.9335 0.0447 -0.0401>, rot=<-0.0037 0.0213 0.298 -0.9543>}
 thumb_03_r {parent=thumb_02_r, loc=<-2.2223 -0.3682 0.0907>, rot=<0.0118 -0.013 0.0066 -0.9998>}
 weapon_r {parent=hand_r, loc=<0.1975 3.4111 -0.3811>, rot=<0 0 0 1>}
 lowerarm_twist_01_r {parent=lowerarm_r, loc=<-17.855 0 0>, rot=<0 0 0 -1>}
 lowerarm_twist_02_r {parent=lowerarm_r, loc=<-9.6199 0 0>, rot=<0 0 0 -1>}
 upperarm_twist_01_r {parent=upperarm_r, loc=<-0.0041 0 0>, rot=<0 0 0 1>}
 upperarm_twist_02_r {parent=upperarm_r, loc=<-13.7942 0 0>, rot=<0 0 0 1>}
 neck_01 {parent=spine_05, loc=<11.3138 0.1529 0>, rot=<0 0 -0.2003 -0.9797>}
 neck_02 {parent=neck_01, loc=<4.5627 0 0>, rot=<0 0 0.0504 -0.9987>}
 head {parent=neck_02, loc=<5.3914 0 0>, rot=<0 0 0.115 -0.9934>}
 faceAttach {parent=head, loc=<-4.1329 -0.9158 0>, rot=<0.0082 -0.7071 -0.0082 -0.7071>, scale=<0.89 0.89 0.89>}
 C_brow_mid {parent=faceAttach, loc=<0.0009 17.5543 16.5664>, rot=<-0.0011 0 0 -1>}
 L_brow_mid {parent=faceAttach, loc=<7.3063 14.1704 16.7506>, rot=<-0.0012 -0.0045 0.2553 -0.9669>}
 L_brow_outer {parent=faceAttach, loc=<9.8557 8.4186 16.3853>, rot=<-0.0296 -0.0404 0.4491 -0.8921>}
 L_cheek_inner {parent=faceAttach, loc=<11.9082 15.1037 3.3102>, rot=<-0.0716 -0.2701 0.3181 -0.906>}
 C_lip_upper_mid {parent=faceAttach, loc=<-0.0184 24.1102 4.5814>, rot=<0.2214 0.0013 -0.0005 -0.9752>}
 L_lip_upper_outer {parent=faceAttach, loc=<5.0105 23.0123 4.2221>, rot=<0.0349 -0.0731 0.2626 -0.9615>}
 L_lip_corner {parent=faceAttach, loc=<7.3367 21.611 3.6412>, rot=<0.0019 0.0079 -0.0146 -0.9999>}
 C_nose_bridge {parent=faceAttach, loc=<-0.0013 21.2334 11.8169>, rot=<-0.2229 0 0 -0.9748>}
 R_brow_mid {parent=faceAttach, loc=<-7.3063 14.1704 16.7506>, rot=<-0.0012 0.0045 -0.2553 -0.9669>}
 R_brow_outer {parent=faceAttach, loc=<-9.8558 8.4186 16.3853>, rot=<-0.0296 0.0404 -0.4491 -0.8921>}
 R_cheek_inner {parent=faceAttach, loc=<-11.9082 15.1037 3.3102>, rot=<-0.0716 0.2701 -0.3181 -0.906>}
 R_lip_corner {parent=faceAttach, loc=<-7.3367 21.611 3.6413>, rot=<0.0019 -0.0079 0.0146 -0.9999>}
 C_jaw {parent=faceAttach, loc=<-0.0151 8.6475 4.2513>, rot=<0 0 0 1>}
 L_lip_lower_outer {parent=C_jaw, loc=<4.9714 14.0166 -0.8941>, rot=<0.012 0.0442 0.3057 -0.951>}
 C_lip_lower_mid {parent=C_jaw, loc=<-0.0066 15.0631 -1.0777>, rot=<-0.0011 0 0 -1>}
 R_lip_lower_outer {parent=C_jaw, loc=<-4.9412 14.0166 -0.8941>, rot=<0.012 -0.0442 -0.3057 -0.951>}
 teeth_lower {parent=C_jaw, loc=<0.0151 -2.0436 -7.8148>, rot=<0 0 0 1>}
 R_eye {parent=faceAttach, loc=<-12.401 12.2287 11.605>, rot=<-0.0095 -0.1612 0.8279 -0.5371>}
 R_eye_lid_upper_mid {parent=R_eye, loc=<-0.606 -0.9187 0.3128>, rot=<0.1665 0.0718 0.238 -0.9542>, scale=<0.9933 1.0112 0.9956>}
 R_eye_lid_lower_mid {parent=R_eye, loc=<-0.3538 -0.8779 -0.6851>, rot=<0.0706 0.0727 0.2442 -0.9644>, scale=<0.9922 1.0088 0.999>}
 L_eye {parent=faceAttach, loc=<13.6423 13.0414 11.9988>, rot=<-0.0191 -0.0413 0.9953 -0.0851>}
 L_eye_lid_upper_mid {parent=L_eye, loc=<-0.4963 -0.9889 0.377>, rot=<0.1556 0.0383 0.2089 -0.9647>}
 L_eye_lid_lower_mid {parent=L_eye, loc=<-0.495 -0.9003 -0.6658>, rot=<0.1104 0.0193 0.2789 -0.9538>}
 R_lip_upper_outer {parent=faceAttach, loc=<-5.0105 23.0123 4.2221>, rot=<0.0349 0.0731 -0.2626 -0.9615>}
 dyn_hatDangle_01_r {parent=head, loc=<7.5967 2.2121 11.7297>, rot=<-0.6523 -0.0949 0.2662 -0.7033>}
 dyn_hatDangle_02_r {parent=dyn_hatDangle_01_r, loc=<-3.9424 0.0003 -0.0001>, rot=<0.2088 0.0205 -0.1134 -0.9712>}
 dyn_hatDangle_03_r {parent=dyn_hatDangle_02_r, loc=<-3.9578 -0.0001 0>, rot=<0.1331 0.0013 -0.0314 -0.9906>}
 dyn_hatDangle_01_l {parent=head, loc=<7.5969 2.2121 -11.7297>, rot=<-0.0949 0.6523 0.7033 0.2662>}
 dyn_hatDangle_02_l {parent=dyn_hatDangle_01_l, loc=<3.9428 0 0>, rot=<0.2088 0.0205 -0.1134 -0.9712>}
 dyn_hatDangle_03_l {parent=dyn_hatDangle_02_l, loc=<3.9576 0 0>, rot=<0.1331 0.0013 -0.0314 -0.9906>}
 dyn_hatclothFront_01 {parent=head, loc=<26.5446 -0.9869 -4.6334>, rot=<-0.1238 0.5389 0.7817 0.2883>}
 dyn_hatclothFront_02 {parent=dyn_hatclothFront_01, loc=<5.0423 0 0>, rot=<0 0 0 1>}
 dyn_hatclothFront_03 {parent=dyn_hatclothFront_02, loc=<5.0423 0 0>, rot=<0 0 0 1>}
 dyn_hatclothMid_01 {parent=head, loc=<25.5576 -3.9152 -4.9421>, rot=<0.1484 -0.8151 -0.5169 -0.2155>}
 dyn_hatclothMid_02 {parent=dyn_hatclothMid_01, loc=<5.5632 0 0>, rot=<0.0079 0.0089 0.0365 -0.9993>}
 dyn_hatclothMid_03 {parent=dyn_hatclothMid_02, loc=<5.7215 0 0>, rot=<0.0071 -0.0051 -0.025 -0.9997>}
 dyn_hatclothBack_01 {parent=head, loc=<25.7198 -5.7754 -2.3375>, rot=<0.1688 -0.9185 -0.3521 -0.0617>}
 dyn_hatclothBack_02 {parent=dyn_hatclothBack_01, loc=<4.788 0 0>, rot=<-0.0149 -0.006 0.0275 -0.9995>}
 dyn_hatclothBack_03 {parent=dyn_hatclothBack_02, loc=<5.0074 0 0>, rot=<0.0377 0.0139 -0.0573 -0.9976>}
 dyn_frontPouch_01 {parent=spine_01, loc=<9.91 8.1201 -7.8536>, rot=<0.2158 -0.2685 -0.7504 -0.5641>}
 dyn_frontPouch_02 {parent=dyn_frontPouch_01, loc=<1.6377 0 0>, rot=<-0.071 -0.0691 -0.5201 -0.8483>}
 dyn_backPouch_01 {parent=spine_01, loc=<10.4884 0.7048 -10.4302>, rot=<0.5108 -0.5832 -0.4884 -0.4007>}
 dyn_backPouch_02 {parent=dyn_backPouch_01, loc=<2.2541 0 0>, rot=<0.0305 0.04 -0.4676 -0.8825>}
 thigh_l {parent=pelvis, loc=<-2.5693 -0.0109 -7.7501>, rot=<0.0012 0.0572 0.0155 -0.9982>}
 calf_l {parent=thigh_l, loc=<-39.0484 0 0>, rot=<0.0006 -0.0161 -0.0391 -0.9991>}
 calf_twist_01_l {parent=calf_l, loc=<-35.8131 -0.0967 0.0397>, rot=<0.0237 0.0095 -0.0035 -0.9997>}
 calf_twist_02_l {parent=calf_l, loc=<-18.8064 -0.0967 0.0397>, rot=<0.0237 0.0095 -0.0035 -0.9997>}
 foot_l {parent=calf_l, loc=<-42.561 0.1077 0.4292>, rot=<0.0418 -0.0455 0.0343 -0.9975>}
 ball_l {parent=foot_l, loc=<-6.1599 13.8264 0.6806>, rot=<0.0001 -0.0001 -0.7063 -0.7079>}
 thigh_twist_01_l {parent=thigh_l, loc=<-21.6303 0 0>, rot=<0.0013 0 0 -1>}
 thigh_r {parent=pelvis, loc=<-2.5691 -0.0109 7.7501>, rot=<0.0572 -0.0012 0.9982 0.0155>}
 calf_r {parent=thigh_r, loc=<39.0487 0 0>, rot=<0.0006 -0.0161 -0.0391 -0.9991>}
 calf_twist_01_r {parent=calf_r, loc=<35.8132 0.0967 -0.0397>, rot=<0.0237 0.0095 -0.0035 -0.9997>}
 calf_twist_02_r {parent=calf_r, loc=<18.8064 0.0967 -0.0397>, rot=<0.0237 0.0095 -0.0035 -0.9997>}
 foot_r {parent=calf_r, loc=<42.5611 -0.1077 -0.4292>, rot=<0.0418 -0.0455 0.0343 -0.9975>}
 ball_r {parent=foot_r, loc=<6.1598 -13.8264 -0.6806>, rot=<0.0001 -0.0001 -0.7063 -0.7079>}
 thigh_twist_01_r {parent=thigh_r, loc=<21.6303 0 0>, rot=<0.0013 0 0 -1>}
 dyn_skirt_mstr {parent=pelvis, loc=<0 0 0>, rot=<0 0 0 1>}
 dyn_skirtFront_01_Base_r {parent=dyn_skirt_mstr, loc=<5.5373 9.8666 6.8263>, rot=<0.0319 -0.1432 0.9571 0.2497>}
 dyn_skirtFront_01_r {parent=dyn_skirtFront_01_Base_r, loc=<0 0 0>, rot=<0 0 0 1>}
 dyn_skirtFront_02_r {parent=dyn_skirtFront_01_r, loc=<10.6006 0 0>, rot=<-0.063 -0.0044 -0.0991 -0.9931>}
 dyn_skirtFront_03_r {parent=dyn_skirtFront_02_r, loc=<19.753 0 0>, rot=<0.0385 0.0287 -0.0653 -0.9967>}
 dyn_skirtFront_04_r {parent=dyn_skirtFront_03_r, loc=<17.1078 0 0>, rot=<0.0021 -0.0145 0.0015 -0.9999>}
 dyn_skirtFront_05_r {parent=dyn_skirtFront_04_r, loc=<12.7909 0 0>, rot=<-0.0264 0.0256 -0.1028 -0.994>}
 dyn_skirtSideFront_01_Base_r {parent=dyn_skirt_mstr, loc=<5.5693 5.5938 15.6766>, rot=<0.1781 -0.4497 0.8607 0.159>}
 dyn_skirtSideFront_01_r {parent=dyn_skirtSideFront_01_Base_r, loc=<0 0 0>, rot=<0 0 0 -1>}
 dyn_skirtSideFront_02_r {parent=dyn_skirtSideFront_01_r, loc=<11.3995 0 0>, rot=<0.1057 -0.0439 -0.1092 -0.9874>}
 dyn_skirtSideFront_03_r {parent=dyn_skirtSideFront_02_r, loc=<19.629 0 0>, rot=<-0.0224 0.0124 -0.0366 -0.999>}
 dyn_skirtSideFront_04_r {parent=dyn_skirtSideFront_03_r, loc=<17.0619 0 0>, rot=<-0.0093 -0.0265 -0.0093 -0.9996>}
 dyn_skirtSideFront_05_r {parent=dyn_skirtSideFront_04_r, loc=<12.8243 0 0>, rot=<-0.0067 0.0489 -0.0079 -0.9987>}
 dyn_skirtBack_01_r {parent=dyn_skirt_mstr, loc=<3.0288 -13.8427 7.2118>, rot=<0.2084 -0.9725 0.0833 0.0619>}
 dyn_skirtBack_02_r {parent=dyn_skirtBack_01_r, loc=<12.8375 0 0>, rot=<0.0264 0.0144 -0.1223 -0.992>}
 dyn_skirtBack_03_r {parent=dyn_skirtBack_02_r, loc=<17.0624 0 0>, rot=<-0.024 0.0143 -0.0321 -0.9991>}
 dyn_skirtBack_04_r {parent=dyn_skirtBack_03_r, loc=<16.546 0 0>, rot=<-0.0694 -0.0095 0.0183 -0.9974>}
 dyn_skirtBack_05_r {parent=dyn_skirtBack_04_r, loc=<12.7673 0 0>, rot=<0.0099 0.0016 -0.05 -0.9987>}
 dyn_skirtSideBack_01_r {parent=dyn_skirt_mstr, loc=<6.273 -5.056 13.804>, rot=<0.2274 -0.8447 0.4631 0.1427>}
 dyn_skirtSideBack_02_r {parent=dyn_skirtSideBack_01_r, loc=<14.4152 0 0>, rot=<0.0293 -0.0298 -0.1719 -0.9842>}
 dyn_skirtSideBack_03_r {parent=dyn_skirtSideBack_02_r, loc=<18.5482 0 0>, rot=<-0.0803 0.0228 -0.0142 -0.9964>}
 dyn_skirtSideBack_04_r {parent=dyn_skirtSideBack_03_r, loc=<17.154 0 0>, rot=<0.0168 -0.0392 0.0275 -0.9987>}
 dyn_skirtSideBack_05_r {parent=dyn_skirtSideBack_04_r, loc=<12.9357 0 0>, rot=<-0.0145 0.0592 -0.0807 -0.9949>}
 dyn_skirtFront_01_Base_l {parent=dyn_skirt_mstr, loc=<5.5373 9.8666 -6.8263>, rot=<0.1432 0.0319 0.2497 -0.9571>}
 dyn_skirtFront_01_l {parent=dyn_skirtFront_01_Base_l, loc=<0 0 0>, rot=<0 0 0 -1>}
 dyn_skirtFront_02_l {parent=dyn_skirtFront_01_l, loc=<-10.6005 0 0>, rot=<-0.063 -0.0044 -0.0991 -0.9931>}
 dyn_skirtFront_03_l {parent=dyn_skirtFront_02_l, loc=<-19.753 0 0>, rot=<0.0385 0.0287 -0.0653 -0.9967>}
 dyn_skirtFront_04_l {parent=dyn_skirtFront_03_l, loc=<-17.1078 -0.0001 0>, rot=<0.0021 -0.0145 0.0015 -0.9999>}
 dyn_skirtFront_05_l {parent=dyn_skirtFront_04_l, loc=<-12.791 0 0>, rot=<-0.0264 0.0256 -0.1028 -0.994>}
 dyn_skirtSideFront_01_Base_l {parent=dyn_skirt_mstr, loc=<5.569 5.594 -15.677>, rot=<0.4497 0.1781 0.159 -0.8607>}
 dyn_skirtSideFront_01_l {parent=dyn_skirtSideFront_01_Base_l, loc=<0 0 0>, rot=<0 0 0 -1>}
 dyn_skirtSideFront_02_l {parent=dyn_skirtSideFront_01_l, loc=<-11.3995 0.0001 0>, rot=<0.1057 -0.0439 -0.1092 -0.9874>}
 dyn_skirtSideFront_03_l {parent=dyn_skirtSideFront_02_l, loc=<-19.629 -0.0001 0>, rot=<-0.0224 0.0124 -0.0366 -0.999>}
 dyn_skirtSideFront_04_l {parent=dyn_skirtSideFront_03_l, loc=<-17.0618 0.0001 0>, rot=<-0.0093 -0.0265 -0.0093 -0.9996>}
 dyn_skirtSideFront_05_l {parent=dyn_skirtSideFront_04_l, loc=<-12.8242 -0.0001 0>, rot=<-0.0067 0.0489 -0.0079 -0.9987>}
 dyn_skirtBack_01_l {parent=dyn_skirt_mstr, loc=<3.029 -13.843 -7.212>, rot=<0.9725 0.2084 0.0619 -0.0833>}
 dyn_skirtBack_02_l {parent=dyn_skirtBack_01_l, loc=<-12.8374 0 0>, rot=<0.0264 0.0144 -0.1223 -0.992>}
 dyn_skirtBack_03_l {parent=dyn_skirtBack_02_l, loc=<-17.0625 0 0>, rot=<-0.024 0.0143 -0.0321 -0.9991>}
 dyn_skirtBack_04_l {parent=dyn_skirtBack_03_l, loc=<-16.5461 0 0>, rot=<-0.0694 -0.0095 0.0183 -0.9974>}
 dyn_skirtBack_05_l {parent=dyn_skirtBack_04_l, loc=<-12.7673 0 0>, rot=<0.0099 0.0016 -0.05 -0.9987>}
 dyn_skirtSideBack_01_l {parent=dyn_skirt_mstr, loc=<6.2727 -5.0562 -13.8039>, rot=<0.8447 0.2274 0.1427 -0.4631>}
 dyn_skirtSideBack_02_l {parent=dyn_skirtSideBack_01_l, loc=<-14.415 0 0>, rot=<0.0293 -0.0298 -0.1719 -0.9842>}
 dyn_skirtSideBack_03_l {parent=dyn_skirtSideBack_02_l, loc=<-18.5481 0 0>, rot=<-0.0803 0.0228 -0.0142 -0.9964>}
 dyn_skirtSideBack_04_l {parent=dyn_skirtSideBack_03_l, loc=<-17.154 0 -0.0001>, rot=<0.0168 -0.0392 0.0275 -0.9987>}
 dyn_skirtSideBack_05_l {parent=dyn_skirtSideBack_04_l, loc=<-12.9357 0 0>, rot=<-0.0145 0.0592 -0.0807 -0.9949>}
 ik_foot_root {parent=root, loc=<0 0 0>, rot=<0 0 0 1>}
 ik_foot_l {parent=ik_foot_root, loc=<15.2608 -4.4176 9.4433>, rot=<-0.0357 -0.7045 -0.0284 0.7082>}
 ik_foot_r {parent=ik_foot_root, loc=<-15.2609 -4.4176 9.4431>, rot=<-0.7082 -0.0284 0.7045 -0.0357>}
 ik_hand_root {parent=root, loc=<0 0 0>, rot=<0 0 0 1>}
 ik_hand_gun {parent=ik_hand_root, loc=<-51.973 4.089 102.694>, rot=<-0.6084 -0.1793 -0.4613 0.6205>}
 ik_hand_l {parent=ik_hand_gun, loc=<53.0284 82.1785 35.2051>, rot=<-0.5102 -0.7906 -0.3387 0>}
 ik_hand_r {parent=ik_hand_gun, loc=<0 0 0>, rot=<0 0 0 1>}
]
variables [crouching {type=double, default=0}]
