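# Physics asset for the Teriyaki Fish Princess female body cosmetic, modeling its skirt, apparently for jumping.
# Eight five-segment box chains ring the skirt front to back, alongside pelvis, thigh and calf shapes for the lower body.
# Swing/twist-limited constraints link segments, though several right-side and root links are absent and only back chains attach to the skirt master.
# Angularly free constraints cross-link some neighboring chains at the fourth segment.
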PhysicsAsset /BRCosmetics/Characters/Player/Female/Medium/Bodies/F_MED_Teriyaki_Fish_Princess/Meshes/F_MED_TeriyakiFish_Princess_JumpPhysics
bodies {
 calf_l [{shapeType=Capsule, length=37.8336, radius=11.4699, center=<-28.3369 0 -0.0011>, rotation=<rot 90 0 -179.9999>}]
 calf_r [{shapeType=Capsule, length=37.8336, radius=11.4699, center=<28.3369 0 0.0011>, rotation=<rot 90 0 0.0001>}]
 dyn_skirtBack_01_l [{shapeType=Box, extentX=1.9782, extentY=19.3183, extentZ=13.5333, center=<-6.113 0.0002 -0.1003>, rotation=<rot 0.2138 89.9996 -85.7057>}]
 dyn_skirtBack_01_r [{shapeType=Box, extentX=1.9782, extentY=19.3183, extentZ=13.5333, center=<6.113 -0.0002 0.1003>, rotation=<rot 0.2138 89.9996 94.2943>}]
 dyn_skirtBack_02_l [{shapeType=Box, extentX=4.3995, extentY=19.4093, extentZ=17.5123, center=<-8.3195 0.0007 0.0769>, rotation=<rot 0 89.9999 -87.2829>}]
 dyn_skirtBack_02_r [{shapeType=Box, extentX=4.3995, extentY=19.4093, extentZ=17.5123, center=<8.3195 -0.0007 -0.0769>, rotation=<rot 0 89.9999 92.7171>}]
 dyn_skirtBack_03_l [{shapeType=Box, extentX=5.7732, extentY=21.5154, extentZ=17.8427, center=<-7.8846 0 -0.0414>, rotation=<rot 0 89.9999 -88.7191>}]
 dyn_skirtBack_03_r [{shapeType=Box, extentX=5.7732, extentY=21.5154, extentZ=17.8427, center=<7.8846 0 0.0414>, rotation=<rot 0 89.9999 91.2809>}]
 dyn_skirtBack_04_l [{shapeType=Box, extentX=3.5925, extentY=23.8766, extentZ=14.09, center=<-5.8362 0 0>, rotation=<rot 0 90 -86.4597>}]
 dyn_skirtBack_04_r [{shapeType=Box, extentX=3.5925, extentY=23.1143, extentZ=14.09, center=<5.8362 0 0>, rotation=<rot 0 90 93.5403>}]
 dyn_skirtBack_05_l [{shapeType=Box, extentX=5.2048, extentY=25.1521, extentZ=17.5724, center=<-7.4211 1.1692 -0.0136>, rotation=<rot -0.6568 79.3248 -86.6123>}]
 dyn_skirtBack_05_r [{shapeType=Box, extentX=5.2048, extentY=24.5909, extentZ=17.5724, center=<7.4211 -1.1692 0.0136>, rotation=<rot -0.6568 79.3248 93.3877>}]
 dyn_skirtFront_01_Base_l [{shapeType=Box, extentX=1.01, extentY=1.01, extentZ=1.01, rotation=<rot 0 90 -90>}]
 dyn_skirtFront_01_Base_r [{shapeType=Box, extentX=1.01, extentY=1.01, extentZ=1.01, rotation=<rot 0 90 90>}]
 dyn_skirtFront_01_l [{shapeType=Box, extentX=1.9782, extentY=17.2533, extentZ=11.6879, center=<-4.7772 -0.0003 0>, rotation=<rot 0 90 -90>}]
 dyn_skirtFront_01_r [{shapeType=Box, extentX=1.9782, extentY=17.2533, extentZ=11.6879, center=<4.7772 0.0003 0>, rotation=<rot 0 90 90>}]
 dyn_skirtFront_02_l [{shapeType=Box, extentX=4.3995, extentY=23.2498, extentZ=20.0926, center=<-9.9396 0.0008 0>, rotation=<rot 0 90 -90>}]
 dyn_skirtFront_02_r [{shapeType=Box, extentX=4.3995, extentY=23.2498, extentZ=20.0926, center=<9.9396 -0.0008 0>, rotation=<rot 0 90 90>}]
 dyn_skirtFront_03_l [{shapeType=Box, extentX=5.7732, extentY=26.5028, extentZ=17.8427, center=<-7.8846 0 -0.0414>, rotation=<rot 0 89.9999 -92.7551>}]
 dyn_skirtFront_03_r [{shapeType=Box, extentX=5.7732, extentY=26.5028, extentZ=17.8427, center=<7.8846 0 0.0414>, rotation=<rot 0 89.9999 87.2449>}]
 dyn_skirtFront_04_l [{shapeType=Box, extentX=3.5925, extentY=29.1413, extentZ=14.09, center=<-5.8362 0 0>, rotation=<rot 0 90 -91.7249>}]
 dyn_skirtFront_04_r [{shapeType=Box, extentX=3.5925, extentY=29.1413, extentZ=14.09, center=<5.8362 0 0>, rotation=<rot 0 90 88.2751>}]
 dyn_skirtFront_05_l [{shapeType=Box, extentX=5.2048, extentY=28.3109, extentZ=17.5724, center=<-7.4211 1.1692 -0.0136>, rotation=<rot -0.6568 79.3266 -93.4811>}]
 dyn_skirtFront_05_r [{shapeType=Box, extentX=5.2048, extentY=28.3109, extentZ=17.5724, center=<7.4211 -1.1692 0.0136>, rotation=<rot -0.6568 79.3266 86.5189>}]
 dyn_skirtSideBack_01_l [{shapeType=Box, extentX=2.9699, extentY=17.2533, extentZ=11.9527, center=<-6.9516 -0.0003 0>, rotation=<rot -0.6357 89.9998 -90.0001>}]
 dyn_skirtSideBack_01_r [{shapeType=Box, extentX=2.9699, extentY=17.2533, extentZ=11.6879, center=<6.9516 0.0003 0>, rotation=<rot -0.6357 89.9998 89.9999>}]
 dyn_skirtSideBack_02_l [{shapeType=Box, extentX=4.3995, extentY=23.2498, extentZ=20.0926, center=<-9.9396 0.0008 0>, rotation=<rot 9.0556 89.9999 -89.9999>}]
 dyn_skirtSideBack_02_r [{shapeType=Box, extentX=4.3995, extentY=23.2498, extentZ=20.0926, center=<9.9396 -0.0008 0>, rotation=<rot 0 90 90>}]
 dyn_skirtSideBack_03_l [{shapeType=Box, extentX=5.7732, extentY=29.0371, extentZ=17.8427, center=<-7.8846 0 -0.0414>, rotation=<rot 0.0679 90.0925 -92.0443>}]
 dyn_skirtSideBack_03_r [{shapeType=Box, extentX=5.7732, extentY=29.0371, extentZ=17.8427, center=<7.8846 0 0.0414>, rotation=<rot -7.6297 90.3693 87.2202>}]
 dyn_skirtSideBack_04_l [{shapeType=Box, extentX=3.5925, extentY=29.1413, extentZ=14.09, center=<-6.1751 0 -0.0127>, rotation=<rot 11.502 90.4055 -87.9639>}]
 dyn_skirtSideBack_04_r [{shapeType=Box, extentX=3.5925, extentY=29.1413, extentZ=14.09, center=<6.1751 0 0.0127>, rotation=<rot 0 89.9998 89.7725>}]
 dyn_skirtSideBack_05_l [{shapeType=Box, extentX=5.2048, extentY=31.3328, extentZ=17.5724, center=<-7.4211 1.1692 -0.0136>, rotation=<rot 4.0903 78.9198 -94.8616>}]
 dyn_skirtSideBack_05_r [{shapeType=Box, extentX=5.2048, extentY=28.3109, extentZ=17.5724, center=<7.4211 -1.1692 0.0136>, rotation=<rot -0.6568 79.3246 86.9085>}]
 dyn_skirtSideFront_01_Base_l [{shapeType=Box, extentX=1.01, extentY=1.01, extentZ=1.01, rotation=<rot 0 90 -90>}]
 dyn_skirtSideFront_01_Base_r [{shapeType=Box, extentX=1.01, extentY=1.01, extentZ=1.01, rotation=<rot 0 90 90>}]
 dyn_skirtSideFront_01_l [{shapeType=Box, extentX=1.9782, extentY=17.2533, extentZ=11.6879, center=<-4.7772 -0.0003 0>, rotation=<rot -13.2003 89.9998 -90.0001>}]
 dyn_skirtSideFront_01_r [{shapeType=Box, extentX=1.9782, extentY=17.2533, extentZ=11.6879, center=<4.7772 0.0003 0>, rotation=<rot -13.2003 89.9999 89.9999>}]
 dyn_skirtSideFront_02_l [{shapeType=Box, extentX=4.3995, extentY=23.2498, extentZ=20.0926, center=<-9.9396 0.0008 0>, rotation=<rot 0 90 -90>}]
 dyn_skirtSideFront_02_r [{shapeType=Box, extentX=4.3995, extentY=23.2498, extentZ=20.0926, center=<9.9396 -0.0008 0>, rotation=<rot 0 90 90>}]
 dyn_skirtSideFront_03_l [{shapeType=Box, extentX=6.8892, extentY=24.6728, extentZ=17.8427, center=<-7.8845 0.2014 -0.0412>, rotation=<rot 0 89.9999 -92.7551>}]
 dyn_skirtSideFront_03_r [{shapeType=Box, extentX=5.7732, extentY=25.3831, extentZ=17.8427, center=<7.8846 0 0.0414>, rotation=<rot 0 89.9999 87.2449>}]
 dyn_skirtSideFront_04_l [{shapeType=Box, extentX=5.4823, extentY=29.1413, extentZ=14.09, center=<-5.8323 0.2014 -0.0034>, rotation=<rot -14.4338 90.444 -91.7811>}]
 dyn_skirtSideFront_04_r [{shapeType=Box, extentX=3.5925, extentY=29.1413, extentZ=14.09, center=<5.8362 0 0>, rotation=<rot 0 90 88.2751>}]
 dyn_skirtSideFront_05_l [{shapeType=Box, extentX=6.3208, extentY=28.3109, extentZ=17.5724, center=<-7.4145 1.3704 -0.0203>, rotation=<rot -13.5955 80.1263 -93.5807>}]
 dyn_skirtSideFront_05_r [{shapeType=Box, extentX=5.2048, extentY=28.3109, extentZ=17.5724, center=<7.4211 -1.1692 0.0136>, rotation=<rot -0.6568 79.3266 86.5189>}]
 dyn_skirt_mstr [{shapeType=Capsule, length=2.3347, radius=3.3975, shapeName=dyn_skirt_mstr_capsule}]
 pelvis [
  {shapeType=Capsule, length=37.2332, radius=19.3827, center=<-81.5848 -7.6952 -0.729>, rotation=<rot -89.4792 0 21.5936>}
  {shapeType=Box, extentX=9.841, extentY=9.841, extentZ=9.841}
 ]
 thigh_l [{shapeType=Capsule, length=39.3364, radius=10.3225, center=<-25.3337 0.0003 0>, rotation=<rot 90 0 0>}]
 thigh_r [{shapeType=Capsule, length=39.3364, radius=10.3225, center=<25.3337 -0.0003 0>, rotation=<rot 90 0 180>}]
 thigh_twist_01_l [{shapeType=Capsule, length=8.2861, radius=6.6718, rotation=<rot 90 0 0>}]
 thigh_twist_01_r [{shapeType=Capsule, length=8.504, radius=6.6597, rotation=<rot 90 0 0>}]
}
constraints [
 {bone1Name=dyn_skirtFront_05_r, bone2Name=dyn_skirtFront_04_r, swing1LimitDegrees=20, swing2LimitDegrees=12, twistLimitDegrees=8}
 {bone1Name=dyn_skirtSideFront_04_r, bone2Name=dyn_skirtSideFront_03_r, swing1LimitDegrees=20, swing2LimitDegrees=12, twistLimitDegrees=8}
 {bone1Name=dyn_skirtSideFront_05_r, bone2Name=dyn_skirtSideFront_04_r, swing1LimitDegrees=20, swing2LimitDegrees=12, twistLimitDegrees=8}
 {bone1Name=dyn_skirtSideBack_03_r, bone2Name=dyn_skirtSideBack_02_r, swing1LimitDegrees=12, swing2LimitDegrees=12, twistLimitDegrees=8}
 {bone1Name=dyn_skirtSideBack_04_r, bone2Name=dyn_skirtSideBack_03_r, swing1LimitDegrees=12, swing2LimitDegrees=12, twistLimitDegrees=8}
 {bone1Name=dyn_skirtSideBack_05_r, bone2Name=dyn_skirtSideBack_04_r, swing1LimitDegrees=12, swing2LimitDegrees=12, twistLimitDegrees=8}
 {bone1Name=dyn_skirtBack_04_r, bone2Name=dyn_skirtBack_03_r, swing1LimitDegrees=12, swing2LimitDegrees=12, twistLimitDegrees=8}
 {bone1Name=dyn_skirtBack_05_r, bone2Name=dyn_skirtBack_04_r, swing1LimitDegrees=12, swing2LimitDegrees=12, twistLimitDegrees=8}
 {
  bone1Name=thigh_l
  bone2Name=pelvis
  swing1LimitDegrees=45
  swing1Motion=Free
  swing2LimitDegrees=45
  swing2Motion=Free
  twistLimitDegrees=45
  twistMotion=Free
 }
 {
  bone1Name=thigh_twist_01_l
  bone2Name=thigh_l
  swing1LimitDegrees=45
  swing1Motion=Free
  swing2LimitDegrees=45
  swing2Motion=Free
  twistLimitDegrees=45
  twistMotion=Free
 }
 {
  bone1Name=thigh_twist_01_r
  bone2Name=thigh_r
  swing1LimitDegrees=45
  swing1Motion=Free
  swing2LimitDegrees=45
  swing2Motion=Free
  twistLimitDegrees=45
  twistMotion=Free
 }
 {bone1Name=dyn_skirtBack_02_l, bone2Name=dyn_skirtBack_01_l, swing1LimitDegrees=12, swing2LimitDegrees=12, twistLimitDegrees=8}
 {bone1Name=dyn_skirtBack_03_l, bone2Name=dyn_skirtBack_02_l, swing1LimitDegrees=12, swing2LimitDegrees=12, twistLimitDegrees=8}
 {bone1Name=dyn_skirtBack_04_l, bone2Name=dyn_skirtBack_03_l, swing1LimitDegrees=12, swing2LimitDegrees=12, twistLimitDegrees=8}
 {bone1Name=dyn_skirtBack_05_l, bone2Name=dyn_skirtBack_04_l, swing1LimitDegrees=12, swing2LimitDegrees=12, twistLimitDegrees=8}
 {bone1Name=dyn_skirtSideBack_03_l, bone2Name=dyn_skirtSideBack_02_l, swing1LimitDegrees=12, swing2LimitDegrees=12, twistLimitDegrees=8}
 {bone1Name=dyn_skirtSideBack_04_l, bone2Name=dyn_skirtSideBack_03_l, swing1LimitDegrees=12, swing2LimitDegrees=12, twistLimitDegrees=8}
 {bone1Name=dyn_skirtSideBack_05_l, bone2Name=dyn_skirtSideBack_04_l, swing1LimitDegrees=12, swing2LimitDegrees=12, twistLimitDegrees=8}
 {bone1Name=dyn_skirtFront_02_l, bone2Name=dyn_skirtFront_01_l, swing1LimitDegrees=20, swing2LimitDegrees=12, twistLimitDegrees=8}
 {bone1Name=dyn_skirtFront_03_l, bone2Name=dyn_skirtFront_02_l, swing1LimitDegrees=25, swing2LimitDegrees=12, twistLimitDegrees=8}
 {bone1Name=dyn_skirtFront_04_l, bone2Name=dyn_skirtFront_03_l, swing1LimitDegrees=25, swing2LimitDegrees=12, twistLimitDegrees=8}
 {bone1Name=dyn_skirtFront_05_l, bone2Name=dyn_skirtFront_04_l, swing1LimitDegrees=25, swing2LimitDegrees=12, twistLimitDegrees=8}
 {bone1Name=dyn_skirtSideFront_02_l, bone2Name=dyn_skirtSideFront_01_l, swing1LimitDegrees=20, swing2LimitDegrees=16, twistLimitDegrees=8}
 {bone1Name=dyn_skirtSideFront_03_l, bone2Name=dyn_skirtSideFront_02_l, swing1LimitDegrees=20, swing2LimitDegrees=16, twistLimitDegrees=8}
 {bone1Name=dyn_skirtSideFront_04_l, bone2Name=dyn_skirtSideFront_03_l, swing1LimitDegrees=20, swing2LimitDegrees=16, twistLimitDegrees=8}
 {bone1Name=dyn_skirtSideFront_05_l, bone2Name=dyn_skirtSideFront_04_l, swing1LimitDegrees=20, swing2LimitDegrees=16, twistLimitDegrees=8}
 {bone1Name=dyn_skirtBack_02_r, bone2Name=dyn_skirtBack_01_r, swing1LimitDegrees=12, swing2LimitDegrees=12, twistLimitDegrees=8}
 {bone1Name=dyn_skirtBack_03_r, bone2Name=dyn_skirtBack_02_r, swing1LimitDegrees=12, swing2LimitDegrees=12, twistLimitDegrees=8}
 {bone1Name=dyn_skirtFront_04_r, bone2Name=dyn_skirtFront_03_r, swing1LimitDegrees=20, swing2LimitDegrees=12, twistLimitDegrees=8}
 {
  bone1Name=thigh_r
  bone2Name=pelvis
  swing1LimitDegrees=45
  swing1Motion=Free
  swing2LimitDegrees=45
  swing2Motion=Free
  twistLimitDegrees=45
  twistMotion=Free
 }
 {bone1Name=calf_r, bone2Name=thigh_r, swing1LimitDegrees=12, swing2LimitDegrees=12, twistLimitDegrees=8}
 {
  bone1Name=calf_l
  bone2Name=thigh_l
  swing1LimitDegrees=45
  swing1Motion=Free
  swing2LimitDegrees=45
  swing2Motion=Free
  twistLimitDegrees=45
  twistMotion=Free
 }
 {
  bone1Name=dyn_skirtBack_04_r
  bone2Name=dyn_skirtBack_04_l
  swing1LimitDegrees=45
  swing1Motion=Free
  swing2LimitDegrees=45
  swing2Motion=Free
  twistLimitDegrees=45
  twistMotion=Free
 }
 {
  bone1Name=dyn_skirtFront_04_l
  bone2Name=dyn_skirtSideFront_04_l
  swing1LimitDegrees=45
  swing1Motion=Free
  swing2LimitDegrees=45
  swing2Motion=Free
  twistLimitDegrees=45
  twistMotion=Free
 }
 {
  bone1Name=dyn_skirtSideBack_04_l
  bone2Name=dyn_skirtSideFront_04_l
  swing1LimitDegrees=45
  swing1Motion=Free
  swing2LimitDegrees=45
  swing2Motion=Free
  twistLimitDegrees=45
  twistMotion=Free
 }
 {bone1Name=dyn_skirtBack_01_l, bone2Name=dyn_skirt_mstr, swing1LimitDegrees=12, swing2LimitDegrees=12, twistLimitDegrees=8}
 {bone1Name=dyn_skirtBack_01_r, bone2Name=dyn_skirt_mstr, swing1LimitDegrees=12, swing2LimitDegrees=12, twistLimitDegrees=8}
]
