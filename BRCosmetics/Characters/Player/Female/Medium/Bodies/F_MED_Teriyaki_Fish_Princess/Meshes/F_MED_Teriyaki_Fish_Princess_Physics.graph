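# Physics asset for the Teriyaki Fish Princess female medium body, defining the bodies and constraints for its skirt physics.
# Pelvis and leg capsules cover the lower body, while boxes form left and right Front, SideFront, SideBack and Back skirt chains of five segments.
# Swing and twist limits are looser at the front, and only back and side-back chains constrain to dyn_skirt_mstr.
# Angularly free cross-links tie some neighboring chains at segments 04 and 05.
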PhysicsAsset /BRCosmetics/Characters/Player/Female/Medium/Bodies/F_MED_Teriyaki_Fish_Princess/Meshes/F_MED_Teriyaki_Fish_Princess_Physics
bodies {
 calf_l [{shapeType=Capsule, length=32.3336, radius=11.9699, center=<-31.2025 0.0004 1.2798>, rotation=<rot 87.0546 179.9851 -0.0076>}]
 calf_r [{shapeType=Capsule, length=31.4586, radius=11.9699, center=<31.3391 0 -2.3852>, rotation=<rot 85.427 179.9998 179.9999>}]
 dyn_skirtBack_01_l [{shapeType=Box, extentX=1.9782, extentY=19.3183, extentZ=13.5333, center=<-6.113 0.0002 -0.1003>, rotation=<rot 0.2138 89.9996 -85.7057>}]
 dyn_skirtBack_01_r [{shapeType=Box, extentX=1.9782, extentY=19.3183, extentZ=13.5333, center=<6.113 -0.0002 0.1003>, rotation=<rot 0.2138 89.9996 94.2943>}]
 dyn_skirtBack_02_l [{shapeType=Box, extentX=4.3995, extentY=19.4093, extentZ=17.5123, center=<-8.3195 0.0007 0.0769>, rotation=<rot 0 89.9999 -87.2829>}]
 dyn_skirtBack_02_r [{shapeType=Box, extentX=4.3995, extentY=19.4093, extentZ=17.5123, center=<8.3195 -0.0007 -0.0769>, rotation=<rot 0 89.9999 92.7171>}]
 dyn_skirtBack_03_l [{shapeType=Box, extentX=5.7732, extentY=40.3904, extentZ=17.8427, center=<-7.8846 0 -0.0414>, rotation=<rot 0 89.9999 -88.7191>}]
 dyn_skirtBack_03_r [{shapeType=Box, extentX=5.7732, extentY=40.3904, extentZ=17.8427, center=<7.8846 0 0.0414>, rotation=<rot 0 89.9999 91.2809>}]
 dyn_skirtBack_04_l [{shapeType=Box, extentX=3.5925, extentY=50.6266, extentZ=23.09, center=<-5.8362 0 0>, rotation=<rot 0 90 -86.4597>}]
 dyn_skirtBack_04_r [{shapeType=Box, extentX=3.5925, extentY=49.8643, extentZ=23.09, center=<5.8362 0 0>, rotation=<rot 0 90 93.5403>}]
 dyn_skirtBack_05_l [{shapeType=Box, extentX=5.2048, extentY=54.0271, extentZ=22.8224, center=<-7.4211 1.1692 -0.0136>, rotation=<rot -0.6568 79.3248 -86.6123>}]
 dyn_skirtBack_05_r [{shapeType=Box, extentX=5.2048, extentY=53.4659, extentZ=22.8224, center=<7.4211 -1.1692 0.0136>, rotation=<rot -0.6568 79.3248 93.3877>}]
 dyn_skirtFront_01_Base_l [{shapeType=Box, extentX=1.01, extentY=1.01, extentZ=1.01, rotation=<rot 0 90 -90>}]
 dyn_skirtFront_01_Base_r [{shapeType=Box, extentX=1.01, extentY=1.01, extentZ=1.01, rotation=<rot 0 90 90>}]
 dyn_skirtFront_01_l [{shapeType=Box, extentX=1.9782, extentY=17.2533, extentZ=11.6879, center=<-4.7772 -0.0003 0>, rotation=<rot 0 90 -90>}]
 dyn_skirtFront_01_r [{shapeType=Box, extentX=1.9782, extentY=17.2533, extentZ=11.6879, center=<4.7772 0.0003 0>, rotation=<rot 0 90 90>}]
 dyn_skirtFront_02_l [{shapeType=Box, extentX=4.3995, extentY=23.2498, extentZ=20.0926, center=<-9.9396 0.0008 0>, rotation=<rot 0 90 -90>}]
 dyn_skirtFront_02_r [{shapeType=Box, extentX=4.3995, extentY=23.2498, extentZ=20.0926, center=<9.9396 -0.0008 0>, rotation=<rot 0 90 90>}]
 dyn_skirtFront_03_l [{shapeType=Box, extentX=5.7732, extentY=26.5028, extentZ=17.8427, center=<-7.8846 0 -0.0414>, rotation=<rot 0 89.9999 -92.7551>}]
 dyn_skirtFront_03_r [{shapeType=Box, extentX=5.7732, extentY=26.5028, extentZ=17.8427, center=<7.8846 0 0.0414>, rotation=<rot 0 89.9999 87.2449>}]
 dyn_skirtFront_04_l [{shapeType=Box, extentX=3.5925, extentY=29.1413, extentZ=14.09, center=<-5.8362 0 0>, rotation=<rot 0 90 -91.7249>}]
 dyn_skirtFront_04_r [{shapeType=Box, extentX=3.5925, extentY=29.1413, extentZ=14.09, center=<5.8362 0 0>, rotation=<rot 0 90 88.2751>}]
 dyn_skirtFront_05_l [{shapeType=Box, extentX=5.2048, extentY=28.3109, extentZ=17.5724, center=<-7.4211 1.1692 -0.0136>, rotation=<rot -0.6568 79.3266 -93.4811>}]
 dyn_skirtFront_05_r [{shapeType=Box, extentX=5.2048, extentY=28.3109, extentZ=17.5724, center=<7.4211 -1.1692 0.0136>, rotation=<rot -0.6568 79.3266 86.5189>}]
 dyn_skirtSideBack_01_l [{shapeType=Box, extentX=2.9699, extentY=17.2533, extentZ=11.6879, center=<-8.9959 4.6247 -0.0513>, rotation=<rot 8.8278 118.602 -90.3089>}]
 dyn_skirtSideBack_01_r [{shapeType=Box, extentX=2.9699, extentY=17.2533, extentZ=11.6879, center=<8.0437 -2.882 0.032>, rotation=<rot -0.5579 118.6523 89.6948>}]
 dyn_skirtSideBack_02_l [{shapeType=Box, extentX=4.3995, extentY=23.2498, extentZ=20.0926, center=<-10.2058 1.6544 0.2636>, rotation=<rot 13.4407 99.2596 -88.5445>}]
 dyn_skirtSideBack_02_r [{shapeType=Box, extentX=4.3995, extentY=23.2498, extentZ=20.0926, center=<9.9396 -0.0008 0>, rotation=<rot 0 90 90>}]
 dyn_skirtSideBack_03_l [{shapeType=Box, extentX=5.7732, extentY=45.4121, extentZ=17.8427, center=<-7.9959 0.7666 -0.0366>, rotation=<rot 6.3421 98.051 -92.0287>}]
 dyn_skirtSideBack_03_r [{shapeType=Box, extentX=5.7732, extentY=42.1621, extentZ=17.8427, center=<7.8876 -0.4744 0.105>, rotation=<rot -7.6297 90.3693 87.2202>}]
 dyn_skirtSideBack_04_l [{shapeType=Box, extentX=3.5925, extentY=45.5163, extentZ=14.09, center=<-6.1725 0 0>, rotation=<rot 10.929 101.1997 -85.8587>}]
 dyn_skirtSideBack_04_r [{shapeType=Box, extentX=3.5925, extentY=42.2663, extentZ=14.09, center=<6.1751 0 0>, rotation=<rot 0 89.9998 89.7725>}]
 dyn_skirtSideBack_05_l [{shapeType=Box, extentX=5.2048, extentY=47.7078, extentZ=17.5724, center=<-7.4211 1.1692 -0.0136>, rotation=<rot 4.0903 78.9198 -94.8616>}]
 dyn_skirtSideBack_05_r [{shapeType=Box, extentX=5.2048, extentY=41.4359, extentZ=17.5724, center=<7.4211 -1.1692 0.0136>, rotation=<rot -0.6568 79.3246 86.9085>}]
 dyn_skirtSideFront_01_Base_l [{shapeType=Box, extentX=1.01, extentY=1.01, extentZ=1.01, rotation=<rot 0 90 -90>}]
 dyn_skirtSideFront_01_Base_r [{shapeType=Box, extentX=1.01, extentY=1.01, extentZ=1.01, rotation=<rot 0 90 90>}]
 dyn_skirtSideFront_01_l [{shapeType=Box, extentX=1.9782, extentY=17.2533, extentZ=11.6879, center=<-4.7772 -0.0003 0>, rotation=<rot -13.2003 89.9998 -90.0001>}]
 dyn_skirtSideFront_01_r [{shapeType=Box, extentX=1.9782, extentY=17.2533, extentZ=11.6879, center=<4.7772 0.0003 0>, rotation=<rot -13.2003 89.9999 89.9999>}]
 dyn_skirtSideFront_02_l [{shapeType=Box, extentX=4.3995, extentY=23.2498, extentZ=20.0926, center=<-10.0079 0.61 0>, rotation=<rot 0 97.2209 -90.0003>}]
 dyn_skirtSideFront_02_r [{shapeType=Box, extentX=4.3995, extentY=23.2498, extentZ=20.0926, center=<9.9396 -0.0008 0>, rotation=<rot 0 90 90>}]
 dyn_skirtSideFront_03_l [{shapeType=Box, extentX=6.3892, extentY=24.6728, extentZ=17.8427, center=<-7.9677 0.7666 -0.0372>, rotation=<rot 0.3885 98.1179 -92.7549>}]
 dyn_skirtSideFront_03_r [{shapeType=Box, extentX=5.7732, extentY=25.3831, extentZ=17.8427, center=<7.8846 0 0.0414>, rotation=<rot 0 89.9999 87.2449>}]
 dyn_skirtSideFront_04_l [{shapeType=Box, extentX=5.4823, extentY=29.1413, extentZ=14.09, center=<-5.8323 0 0>, rotation=<rot -2.5543 98.4215 -90.2858>}]
 dyn_skirtSideFront_04_r [{shapeType=Box, extentX=3.5925, extentY=29.1413, extentZ=14.09, center=<5.8362 0 0>, rotation=<rot 0 90 88.2751>}]
 dyn_skirtSideFront_05_l [{shapeType=Box, extentX=6.3208, extentY=28.3109, extentZ=17.5724, center=<-7.4145 1.3704 -0.0203>, rotation=<rot -1.2449 94.9242 -96.8911>}]
 dyn_skirtSideFront_05_r [{shapeType=Box, extentX=5.2048, extentY=28.3109, extentZ=17.5724, center=<7.4211 -1.1692 0.0136>, rotation=<rot -0.6568 79.3266 86.5189>}]
 dyn_skirt_mstr [{shapeType=Capsule, length=1.2958, radius=3.3375, shapeName=dyn_skirt_mstr_capsule}]
 pelvis [
  {shapeType=Capsule, length=37.4832, radius=19.1327, center=<-82.2281 -2.8225 -0.7349>, rotation=<rot -89.4792 -0.0005 3.4151>}
  {shapeType=Box, extentX=9.841, extentY=9.841, extentZ=9.841}
 ]
 thigh_l [{shapeType=Capsule, length=39.3364, radius=10.3225, center=<-25.3337 0.0003 0>, rotation=<rot 90 0 0>}]
 thigh_r [{shapeType=Capsule, length=39.3364, radius=10.3225, center=<25.3337 -0.0003 0>, rotation=<rot 90 0 180>}]
 thigh_twist_01_l [{shapeType=Capsule, length=8.2861, radius=6.6718, rotation=<rot 90 0 0>}]
 thigh_twist_01_r [{shapeType=Capsule, length=8.504, radius=6.6597, rotation=<rot 90 0 0>}]
}
constraints [
 {bone1Name=dyn_skirtFront_05_r, bone2Name=dyn_skirtFront_04_r, swing1LimitDegrees=20, swing2LimitDegrees=12, twistLimitDegrees=8}
 {bone1Name=dyn_skirtSideFront_02_r, bone2Name=dyn_skirtSideFront_01_r, swing1LimitDegrees=20, swing2LimitDegrees=12, twistLimitDegrees=8}
 {bone1Name=dyn_skirtSideFront_03_r, bone2Name=dyn_skirtSideFront_02_r, swing1LimitDegrees=20, swing2LimitDegrees=12, twistLimitDegrees=8}
 {bone1Name=dyn_skirtSideFront_04_r, bone2Name=dyn_skirtSideFront_03_r, swing1LimitDegrees=20, swing2LimitDegrees=12, twistLimitDegrees=8}
 {bone1Name=dyn_skirtSideFront_05_r, bone2Name=dyn_skirtSideFront_04_r, swing1LimitDegrees=20, swing2LimitDegrees=12, twistLimitDegrees=8}
 {bone1Name=dyn_skirtSideBack_02_r, bone2Name=dyn_skirtSideBack_01_r, swing1LimitDegrees=12, swing2LimitDegrees=12, twistLimitDegrees=8}
 {bone1Name=dyn_skirtSideBack_03_r, bone2Name=dyn_skirtSideBack_02_r, swing1LimitDegrees=12, swing2LimitDegrees=12, twistLimitDegrees=8}
 {bone1Name=dyn_skirtSideBack_04_r, bone2Name=dyn_skirtSideBack_03_r, swing1LimitDegrees=12, swing2LimitDegrees=12, twistLimitDegrees=8}
 {bone1Name=dyn_skirtSideBack_05_r, bone2Name=dyn_skirtSideBack_04_r, swing1LimitDegrees=12, swing2LimitDegrees=12, twistLimitDegrees=8}
 {bone1Name=dyn_skirtBack_04_r, bone2Name=dyn_skirtBack_03_r, swing1LimitDegrees=12, swing2LimitDegrees=12, twistLimitDegrees=8}
 {bone1Name=dyn_skirtBack_05_r, bone2Name=dyn_skirtBack_04_r, swing1LimitDegrees=12, swing2LimitDegrees=12, twistLimitDegrees=8}
 {bone1Name=dyn_skirtBack_02_l, bone2Name=dyn_skirtBack_01_l, swing1LimitDegrees=12, swing2LimitDegrees=12, twistLimitDegrees=8}
 {bone1Name=dyn_skirtBack_03_l, bone2Name=dyn_skirtBack_02_l, swing1LimitDegrees=12, swing2LimitDegrees=12, twistLimitDegrees=8}
 {bone1Name=dyn_skirtBack_04_l, bone2Name=dyn_skirtBack_03_l, swing1LimitDegrees=12, swing2LimitDegrees=12, twistLimitDegrees=8}
 {bone1Name=dyn_skirtBack_05_l, bone2Name=dyn_skirtBack_04_l, swing1LimitDegrees=12, swing2LimitDegrees=12, twistLimitDegrees=8}
 {bone1Name=dyn_skirtSideBack_02_l, bone2Name=dyn_skirtSideBack_01_l, swing1LimitDegrees=12, swing2LimitDegrees=12, twistLimitDegrees=8}
 {bone1Name=dyn_skirtSideBack_03_l, bone2Name=dyn_skirtSideBack_02_l, swing1LimitDegrees=12, swing2LimitDegrees=12, twistLimitDegrees=8}
 {bone1Name=dyn_skirtSideBack_04_l, bone2Name=dyn_skirtSideBack_03_l, swing1LimitDegrees=12, swing2LimitDegrees=12, twistLimitDegrees=8}
 {bone1Name=dyn_skirtSideBack_05_l, bone2Name=dyn_skirtSideBack_04_l, swing1LimitDegrees=12, swing2LimitDegrees=12, twistLimitDegrees=8}
 {bone1Name=dyn_skirtFront_02_l, bone2Name=dyn_skirtFront_01_l, swing1LimitDegrees=20, swing2LimitDegrees=12, twistLimitDegrees=8}
 {bone1Name=dyn_skirtFront_03_l, bone2Name=dyn_skirtFront_02_l, swing1LimitDegrees=25, swing2LimitDegrees=12, twistLimitDegrees=8}
 {bone1Name=dyn_skirtFront_04_l, bone2Name=dyn_skirtFront_03_l, swing1LimitDegrees=25, swing2LimitDegrees=12, twistLimitDegrees=8}
 {bone1Name=dyn_skirtFront_05_l, bone2Name=dyn_skirtFront_04_l, swing1LimitDegrees=25, swing2LimitDegrees=12, twistLimitDegrees=8}
 {bone1Name=dyn_skirtSideFront_02_l, bone2Name=dyn_skirtSideFront_01_l, swing1LimitDegrees=20, swing2LimitDegrees=16, twistLimitDegrees=8}
 {bone1Name=dyn_skirtSideFront_03_l, bone2Name=dyn_skirtSideFront_02_l, swing1LimitDegrees=20, swing2LimitDegrees=16, twistLimitDegrees=8}
 {bone1Name=dyn_skirtSideFront_04_l, bone2Name=dyn_skirtSideFront_03_l, swing1LimitDegrees=20, swing2LimitDegrees=16, twistLimitDegrees=8}
 {bone1Name=dyn_skirtSideFront_05_l, bone2Name=dyn_skirtSideFront_04_l, swing1LimitDegrees=20, swing2LimitDegrees=16, twistLimitDegrees=8}
 {bone1Name=dyn_skirtBack_02_r, bone2Name=dyn_skirtBack_01_r, swing1LimitDegrees=12, swing2LimitDegrees=12, twistLimitDegrees=8}
 {bone1Name=dyn_skirtBack_03_r, bone2Name=dyn_skirtBack_02_r, swing1LimitDegrees=12, swing2LimitDegrees=12, twistLimitDegrees=8}
 {bone1Name=dyn_skirtFront_02_r, bone2Name=dyn_skirtFront_01_r, swing1LimitDegrees=20, swing2LimitDegrees=16, twistLimitDegrees=8}
 {bone1Name=dyn_skirtFront_03_r, bone2Name=dyn_skirtFront_02_r, swing1LimitDegrees=20, swing2LimitDegrees=12, twistLimitDegrees=8}
 {bone1Name=dyn_skirtFront_04_r, bone2Name=dyn_skirtFront_03_r, swing1LimitDegrees=20, swing2LimitDegrees=12, twistLimitDegrees=8}
 {
  bone1Name=dyn_skirtBack_05_r
  bone2Name=dyn_skirtBack_05_l
  swing1LimitDegrees=45
  swing1Motion=Free
  swing2LimitDegrees=45
  swing2Motion=Free
  twistLimitDegrees=45
  twistMotion=Free
 }
 {
  bone1Name=dyn_skirtFront_04_l
  bone2Name=dyn_skirtSideFront_04_l
  swing1LimitDegrees=45
  swing1Motion=Free
  swing2LimitDegrees=45
  swing2Motion=Free
  twistLimitDegrees=45
  twistMotion=Free
 }
 {bone1Name=dyn_skirtBack_01_l, bone2Name=dyn_skirt_mstr, swing1LimitDegrees=12, swing2LimitDegrees=12, twistLimitDegrees=8}
 {bone1Name=dyn_skirtBack_01_r, bone2Name=dyn_skirt_mstr, swing1LimitDegrees=12, swing2LimitDegrees=12, twistLimitDegrees=8}
 {bone1Name=dyn_skirtSideBack_01_r, bone2Name=dyn_skirt_mstr, swing1LimitDegrees=12, swing2LimitDegrees=12, twistLimitDegrees=8}
 {
  bone1Name=dyn_skirtSideFront_05_r
  bone2Name=dyn_skirtSideBack_05_r
  swing1LimitDegrees=45
  swing1Motion=Free
  swing2LimitDegrees=45
  swing2Motion=Free
  twistLimitDegrees=45
  twistMotion=Free
 }
 {
  bone1Name=dyn_skirtSideFront_05_l
  bone2Name=dyn_skirtSideBack_05_l
  swing1LimitDegrees=45
  swing1Motion=Free
  swing2LimitDegrees=45
  swing2Motion=Free
  twistLimitDegrees=45
  twistMotion=Free
 }
 {
  bone1Name=dyn_skirtBack_04_l
  bone2Name=dyn_skirtSideBack_04_l
  swing1LimitDegrees=45
  swing1Motion=Free
  swing2LimitDegrees=45
  swing2Motion=Free
  twistLimitDegrees=45
  twistMotion=Free
 }
 {
  bone1Name=dyn_skirtBack_04_r
  bone2Name=dyn_skirtSideBack_04_r
  swing1LimitDegrees=45
  swing1Motion=Free
  swing2LimitDegrees=45
  swing2Motion=Free
  twistLimitDegrees=45
  twistMotion=Free
 }
 {bone1Name=dyn_skirtSideBack_01_l, bone2Name=dyn_skirt_mstr, swing1LimitDegrees=12, swing2LimitDegrees=12, twistLimitDegrees=8}
]
