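# Physics asset for the Teriyaki Fish Princess body's skirt, with a GrindRail name suggesting a rail-grinding variant.
# Box bodies form left and right front, side and back skirt chains alongside leg capsules, a pelvis box and a large root sphere.
# Skirt chains use tight swing and twist limits, generally looser at the front, while most body joints are free.
# Tuning is asymmetric, with extra angled boxes on the lower left side-front chain.
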PhysicsAsset /BRCosmetics/Characters/Player/Female/Medium/Bodies/F_MED_Teriyaki_Fish_Princess/Meshes/F_MED_Teriyaki_Princess_GrindRail_Physics
bodies {
 calf_l [{shapeType=Capsule, length=38.0866, radius=11.4699, center=<-24.204 0 -0.001>, rotation=<rot 90 0 -179.9999>}]
 calf_r [{shapeType=Capsule, length=42.5773, radius=11.4699, center=<25.0136 0 0.001>, rotation=<rot 90 0 0.0001>}]
 dyn_skirtBack_01_l [{shapeType=Box, extentX=5.4782, extentY=40.8183, extentZ=17.0333, center=<-6.113 0.0002 -0.1003>, rotation=<rot 0.2138 89.9996 -85.7057>}]
 dyn_skirtBack_01_r [{shapeType=Box, extentX=5.4782, extentY=40.8183, extentZ=17.0333, center=<6.113 -0.0002 0.1003>, rotation=<rot 0.2138 89.9996 94.2943>}]
 dyn_skirtBack_02_l [{shapeType=Box, extentX=7.8995, extentY=40.9093, extentZ=21.0123, center=<-8.3195 0.0007 0.0769>, rotation=<rot 0 89.9999 -87.2829>}]
 dyn_skirtBack_02_r [{shapeType=Box, extentX=7.8995, extentY=40.9093, extentZ=21.0123, center=<8.3195 -0.0007 -0.0769>, rotation=<rot 0 89.9999 92.7171>}]
 dyn_skirtBack_03_l [{shapeType=Box, extentX=9.2732, extentY=44.6754, extentZ=21.9837, center=<-7.8846 0 -0.0414>, rotation=<rot 0 89.9999 -88.7191>}]
 dyn_skirtBack_03_r [{shapeType=Box, extentX=9.2732, extentY=44.6754, extentZ=21.9837, center=<7.8846 0 0.0414>, rotation=<rot 0 89.9999 91.2809>}]
 dyn_skirtBack_04_l [{shapeType=Box, extentX=7.0925, extentY=47.5522, extentZ=18.9692, center=<-5.8362 0 0>, rotation=<rot 0 90 -86.4597>}]
 dyn_skirtBack_04_r [{shapeType=Box, extentX=7.0925, extentY=46.7899, extentZ=18.9692, center=<5.8362 0 0>, rotation=<rot 0 90 93.5403>}]
 dyn_skirtBack_05_l [{shapeType=Box, extentX=8.7048, extentY=50.8668, extentZ=22.011, center=<-7.4211 1.1692 -0.0136>, rotation=<rot -0.6568 79.3248 -86.6123>}]
 dyn_skirtBack_05_r [{shapeType=Box, extentX=8.7048, extentY=50.3055, extentZ=22.011, center=<7.4211 -1.1692 0.0136>, rotation=<rot -0.6568 79.3248 93.3877>}]
 dyn_skirtFront_01_Base_l [{shapeType=Box, extentX=1.01, extentY=1.01, extentZ=1.01, rotation=<rot 0 90 -90>}]
 dyn_skirtFront_01_Base_r [{shapeType=Box, extentX=1.01, extentY=1.01, extentZ=1.01, rotation=<rot 0 90 90>}]
 dyn_skirtFront_01_l [{shapeType=Box, extentX=1.9782, extentY=17.2533, extentZ=11.6879, center=<-4.7772 -0.0003 0>, rotation=<rot 0 90 -90>}]
 dyn_skirtFront_01_r [{shapeType=Box, extentX=1.9782, extentY=17.2533, extentZ=11.6879, center=<4.7772 0.0003 0>, rotation=<rot 0 90 90>}]
 dyn_skirtFront_02_l [{shapeType=Box, extentX=7.8995, extentY=44.7498, extentZ=23.5926, center=<-9.9396 0.0008 0>, rotation=<rot 0 90 -90>}]
 dyn_skirtFront_02_r [{shapeType=Box, extentX=7.8995, extentY=44.7498, extentZ=23.5926, center=<9.9396 -0.0008 0>, rotation=<rot 0 90 90>}]
 dyn_skirtFront_03_l [{shapeType=Box, extentX=9.2732, extentY=48.0028, extentZ=21.3427, center=<-7.8846 0 -0.0414>, rotation=<rot 0 89.9999 -92.7551>}]
 dyn_skirtFront_03_r [{shapeType=Box, extentX=9.2732, extentY=48.0028, extentZ=21.3427, center=<7.8846 0 0.0414>, rotation=<rot 0 89.9999 87.2449>}]
 dyn_skirtFront_04_l [{shapeType=Box, extentX=7.0925, extentY=50.6413, extentZ=17.59, center=<-5.8362 0 0>, rotation=<rot 0 90 -91.7249>}]
 dyn_skirtFront_04_r [{shapeType=Box, extentX=7.0925, extentY=50.6413, extentZ=17.59, center=<5.8362 0 0>, rotation=<rot 0 90 88.2751>}]
 dyn_skirtFront_05_l [{shapeType=Box, extentX=8.7048, extentY=49.8109, extentZ=21.0724, center=<-7.4211 1.1692 -0.0136>, rotation=<rot -0.6568 79.3266 -93.4811>}]
 dyn_skirtFront_05_r [{shapeType=Box, extentX=8.7048, extentY=49.8109, extentZ=21.0724, center=<7.4211 -1.1692 0.0136>, rotation=<rot -0.6568 79.3266 86.5189>}]
 dyn_skirtSideBack_01_l [{shapeType=Box, extentX=2.9699, extentY=17.2533, extentZ=11.6879, center=<-6.9516 -0.0003 0>, rotation=<rot -0.6357 89.9998 -90.0001>}]
 dyn_skirtSideBack_01_r [{shapeType=Box, extentX=2.9699, extentY=17.2533, extentZ=11.6879, center=<6.9516 0.0003 0>, rotation=<rot -0.6357 89.9998 89.9999>}]
 dyn_skirtSideBack_02_l [{shapeType=Box, extentX=4.3995, extentY=41.2498, extentZ=20.0926, center=<-9.9396 0.0008 0>, rotation=<rot 9.0556 89.9999 -89.9999>}]
 dyn_skirtSideBack_02_r [{shapeType=Box, extentX=7.8995, extentY=44.7498, extentZ=23.5926, center=<9.9396 -0.0008 0>, rotation=<rot 0 90 90>}]
 dyn_skirtSideBack_03_l [{shapeType=Box, extentX=9.2732, extentY=50.5371, extentZ=21.3427, center=<-7.8846 0 -0.0414>, rotation=<rot 0.0679 90.0925 -92.0443>}]
 dyn_skirtSideBack_03_r [{shapeType=Box, extentX=9.2732, extentY=50.5371, extentZ=21.3427, center=<7.8846 0 0.0414>, rotation=<rot -7.6297 90.3693 87.2202>}]
 dyn_skirtSideBack_04_l [{shapeType=Box, extentX=7.0925, extentY=50.6413, extentZ=17.59, center=<-6.1751 0 -0.0127>, rotation=<rot 2.0966 90.0727 -88.0038>}]
 dyn_skirtSideBack_04_r [{shapeType=Box, extentX=7.0925, extentY=50.6413, extentZ=17.59, center=<6.1751 0 0.0127>, rotation=<rot 0 89.9998 89.7725>}]
 dyn_skirtSideBack_05_l [{shapeType=Box, extentX=8.7048, extentY=52.8328, extentZ=21.0724, center=<-7.4211 1.1692 -0.0136>, rotation=<rot -0.6568 79.3238 -94.8495>}]
 dyn_skirtSideBack_05_r [{shapeType=Box, extentX=8.7048, extentY=49.8109, extentZ=21.0724, center=<7.4211 -1.1692 0.0136>, rotation=<rot -0.6568 79.3246 86.9085>}]
 dyn_skirtSideFront_01_Base_l [{shapeType=Box, extentX=1.01, extentY=1.01, extentZ=1.01, rotation=<rot 0 90 -90>}]
 dyn_skirtSideFront_01_Base_r [{shapeType=Box, extentX=1.01, extentY=1.01, extentZ=1.01, rotation=<rot 0 90 90>}]
 dyn_skirtSideFront_01_l [{shapeType=Box, extentX=1.9782, extentY=17.2533, extentZ=11.6879, center=<-4.7772 -0.0003 0>, rotation=<rot -13.2003 89.9998 -90.0001>}]
 dyn_skirtSideFront_01_r [{shapeType=Box, extentX=1.9782, extentY=17.2533, extentZ=11.6879, center=<4.7772 0.0003 0>, rotation=<rot -13.2003 89.9999 89.9999>}]
 dyn_skirtSideFront_02_l [{shapeType=Box, extentX=4.3995, extentY=23.2498, extentZ=20.0926, center=<-9.9396 0.0008 0>, rotation=<rot 0 90 -90>}]
 dyn_skirtSideFront_02_r [{shapeType=Box, extentX=7.8995, extentY=44.7498, extentZ=23.5926, center=<9.9396 -0.0008 0>, rotation=<rot 0 90 90>}]
 dyn_skirtSideFront_03_l [
  {shapeType=Box, extentX=10.3892, extentY=46.1728, extentZ=21.3427, center=<-7.8845 0.2014 -0.0412>, rotation=<rot 0 89.9999 -92.7551>}
  {shapeType=Box, extentX=16.6056, extentY=4.6236, extentZ=19.8801, center=<-6.6882 2.0092 14.1029>, rotation=<rot -3.3635 0 -21.0936>}
 ]
 dyn_skirtSideFront_03_r [{shapeType=Box, extentX=9.2732, extentY=46.8831, extentZ=21.3427, center=<7.8846 0 0.0414>, rotation=<rot 0 89.9999 87.2449>}]
 dyn_skirtSideFront_04_l [
  {shapeType=Box, extentX=8.9823, extentY=50.6413, extentZ=17.59, center=<-5.8323 0.2014 -0.0034>, rotation=<rot 0 90 -91.7249>}
  {shapeType=Box, extentX=12.8193, extentY=3.9405, extentZ=19.1169, center=<-5.7967 3.0175 14.2596>, rotation=<rot 0 0 -17.5526>}
 ]
 dyn_skirtSideFront_04_r [{shapeType=Box, extentX=7.0925, extentY=50.6413, extentZ=17.59, center=<5.8362 0 0>, rotation=<rot 0 90 88.2751>}]
 dyn_skirtSideFront_05_l [
  {shapeType=Box, extentX=9.8208, extentY=49.8109, extentZ=21.0724, center=<-7.4145 1.3704 -0.0203>, rotation=<rot -0.6568 79.3258 -93.4809>}
  {shapeType=Box, extentX=17.3578, extentY=5.5145, extentZ=22.5674, center=<-6.0116 3.4378 15.1101>, rotation=<rot -8.5955 3.0007 -20.6227>}
 ]
 dyn_skirtSideFront_05_r [{shapeType=Box, extentX=8.7048, extentY=49.8109, extentZ=21.0724, center=<7.4211 -1.1692 0.0136>, rotation=<rot -0.6568 79.3266 86.5189>}]
 dyn_skirt_mstr [{shapeType=Capsule, length=4.6981, radius=3.6584, shapeName=dyn_skirt_mstr_capsule}]
 pelvis [{shapeType=Box, extentX=9.841, extentY=9.841, extentZ=9.841}]
 root [{shapeType=Sphere, radius=120, center=<0.002 -12.1883 -96.8818>}]
 thigh_l [{shapeType=Capsule, length=39.3364, radius=10.3225, center=<-25.3337 0.0003 0>, rotation=<rot 90 0 0>}]
 thigh_r [{shapeType=Capsule, length=39.3364, radius=10.3225, center=<25.3337 -0.0003 0>, rotation=<rot 90 0 180>}]
 thigh_twist_01_l [{shapeType=Capsule, length=8.2861, radius=6.6718, rotation=<rot 90 0 0>}]
 thigh_twist_01_r [{shapeType=Capsule, length=8.504, radius=6.6597, rotation=<rot 90 0 0>}]
}
constraints [
 {bone1Name=dyn_skirtFront_05_r, bone2Name=dyn_skirtFront_04_r, swing1LimitDegrees=35, swing2LimitDegrees=12, twistLimitDegrees=8}
 {bone1Name=dyn_skirtSideFront_02_r, bone2Name=dyn_skirtSideFront_01_r, swing1LimitDegrees=20, swing2LimitDegrees=12, twistLimitDegrees=8}
 {bone1Name=dyn_skirtSideFront_03_r, bone2Name=dyn_skirtSideFront_02_r, swing1LimitDegrees=20, swing2LimitDegrees=12, twistLimitDegrees=8}
 {bone1Name=dyn_skirtSideFront_04_r, bone2Name=dyn_skirtSideFront_03_r, swing1LimitDegrees=20, swing2LimitDegrees=12, twistLimitDegrees=8}
 {bone1Name=dyn_skirtSideFront_05_r, bone2Name=dyn_skirtSideFront_04_r, swing1LimitDegrees=20, swing2LimitDegrees=12, twistLimitDegrees=8}
 {bone1Name=dyn_skirtSideBack_02_r, bone2Name=dyn_skirtSideBack_01_r, swing1LimitDegrees=12, swing2LimitDegrees=12, twistLimitDegrees=8}
 {bone1Name=dyn_skirtSideBack_03_r, bone2Name=dyn_skirtSideBack_02_r, swing1LimitDegrees=12, swing2LimitDegrees=12, twistLimitDegrees=8}
 {bone1Name=dyn_skirtSideBack_04_r, bone2Name=dyn_skirtSideBack_03_r, swing1LimitDegrees=12, swing2LimitDegrees=12, twistLimitDegrees=8}
 {bone1Name=dyn_skirtSideBack_05_r, bone2Name=dyn_skirtSideBack_04_r, swing1LimitDegrees=12, swing2LimitDegrees=12, twistLimitDegrees=8}
 {bone1Name=dyn_skirtBack_04_r, bone2Name=dyn_skirtBack_03_r, swing1LimitDegrees=12, swing2LimitDegrees=12, twistLimitDegrees=8}
 {bone1Name=dyn_skirtBack_05_r, bone2Name=dyn_skirtBack_04_r, swing1LimitDegrees=12, swing2LimitDegrees=12, twistLimitDegrees=8}
 {
  bone1Name=thigh_l
  bone2Name=pelvis
  swing1LimitDegrees=45
  swing1Motion=Free
  swing2LimitDegrees=45
  swing2Motion=Free
  twistLimitDegrees=45
  twistMotion=Free
 }
 {
  bone1Name=thigh_twist_01_l
  bone2Name=thigh_l
  swing1LimitDegrees=45
  swing1Motion=Free
  swing2LimitDegrees=45
  swing2Motion=Free
  twistLimitDegrees=45
  twistMotion=Free
 }
 {
  bone1Name=thigh_twist_01_r
  bone2Name=thigh_r
  swing1LimitDegrees=45
  swing1Motion=Free
  swing2LimitDegrees=45
  swing2Motion=Free
  twistLimitDegrees=45
  twistMotion=Free
 }
 {bone1Name=dyn_skirtBack_02_l, bone2Name=dyn_skirtBack_01_l, swing1LimitDegrees=12, swing2LimitDegrees=12, twistLimitDegrees=8}
 {bone1Name=dyn_skirtBack_03_l, bone2Name=dyn_skirtBack_02_l, swing1LimitDegrees=12, swing2LimitDegrees=12, twistLimitDegrees=8}
 {bone1Name=dyn_skirtBack_04_l, bone2Name=dyn_skirtBack_03_l, swing1LimitDegrees=12, swing2LimitDegrees=12, twistLimitDegrees=8}
 {bone1Name=dyn_skirtBack_05_l, bone2Name=dyn_skirtBack_04_l, swing1LimitDegrees=12, swing2LimitDegrees=12, twistLimitDegrees=8}
 {bone1Name=dyn_skirtSideBack_02_l, bone2Name=dyn_skirtSideBack_01_l, swing1LimitDegrees=12, swing2LimitDegrees=12, twistLimitDegrees=8}
 {bone1Name=dyn_skirtSideBack_03_l, bone2Name=dyn_skirtSideBack_02_l, swing1LimitDegrees=12, swing2LimitDegrees=12, twistLimitDegrees=8}
 {bone1Name=dyn_skirtSideBack_04_l, bone2Name=dyn_skirtSideBack_03_l, swing1LimitDegrees=12, swing2LimitDegrees=12, twistLimitDegrees=8}
 {bone1Name=dyn_skirtSideBack_05_l, bone2Name=dyn_skirtSideBack_04_l, swing1LimitDegrees=12, swing2LimitDegrees=12, twistLimitDegrees=8}
 {bone1Name=dyn_skirtFront_02_l, bone2Name=dyn_skirtFront_01_l, swing1LimitDegrees=20, swing2LimitDegrees=12, twistLimitDegrees=8}
 {bone1Name=dyn_skirtFront_03_l, bone2Name=dyn_skirtFront_02_l, swing1LimitDegrees=25, swing2LimitDegrees=12, twistLimitDegrees=8}
 {bone1Name=dyn_skirtFront_04_l, bone2Name=dyn_skirtFront_03_l, swing1LimitDegrees=25, swing2LimitDegrees=12, twistLimitDegrees=8}
 {bone1Name=dyn_skirtFront_05_l, bone2Name=dyn_skirtFront_04_l, swing1LimitDegrees=25, swing2LimitDegrees=12, twistLimitDegrees=8}
 {bone1Name=dyn_skirtSideFront_03_l, bone2Name=dyn_skirtSideFront_02_l, swing1LimitDegrees=20, swing2LimitDegrees=16, twistLimitDegrees=8}
 {bone1Name=dyn_skirtSideFront_04_l, bone2Name=dyn_skirtSideFront_03_l, swing1LimitDegrees=20, swing2LimitDegrees=16, twistLimitDegrees=8}
 {bone1Name=dyn_skirtSideFront_05_l, bone2Name=dyn_skirtSideFront_04_l, swing1LimitDegrees=20, swing2LimitDegrees=16, twistLimitDegrees=8}
 {
  bone1Name=calf_l
  bone2Name=thigh_l
  swing1LimitDegrees=45
  swing1Motion=Free
  swing2LimitDegrees=45
  swing2Motion=Free
  twistLimitDegrees=45
  twistMotion=Free
 }
 {bone1Name=dyn_skirtBack_02_r, bone2Name=dyn_skirtBack_01_r, swing1LimitDegrees=12, swing2LimitDegrees=12, twistLimitDegrees=8}
 {bone1Name=dyn_skirtBack_03_r, bone2Name=dyn_skirtBack_02_r, swing1LimitDegrees=12, swing2LimitDegrees=12, twistLimitDegrees=8}
 {bone1Name=dyn_skirtFront_02_r, bone2Name=dyn_skirtFront_01_r, swing1LimitDegrees=35, swing2LimitDegrees=16, twistLimitDegrees=8}
 {bone1Name=dyn_skirtFront_03_r, bone2Name=dyn_skirtFront_02_r, swing1LimitDegrees=35, swing2LimitDegrees=12, twistLimitDegrees=8}
 {bone1Name=dyn_skirtFront_04_r, bone2Name=dyn_skirtFront_03_r, swing1LimitDegrees=35, swing2LimitDegrees=12, twistLimitDegrees=8}
 {
  bone1Name=thigh_r
  bone2Name=pelvis
  swing1LimitDegrees=45
  swing1Motion=Free
  swing2LimitDegrees=45
  swing2Motion=Free
  twistLimitDegrees=45
  twistMotion=Free
 }
 {
  bone1Name=pelvis
  bone2Name=root
  swing1LimitDegrees=45
  swing1Motion=Free
  swing2LimitDegrees=45
  swing2Motion=Free
  twistLimitDegrees=45
  twistMotion=Free
 }
 {bone1Name=calf_r, bone2Name=thigh_r, swing1LimitDegrees=12, swing2LimitDegrees=12, twistLimitDegrees=8}
 {
  bone1Name=dyn_skirtBack_04_r
  bone2Name=dyn_skirtBack_04_l
  swing1LimitDegrees=45
  swing1Motion=Free
  swing2LimitDegrees=45
  swing2Motion=Free
  twistLimitDegrees=45
  twistMotion=Free
 }
 {bone1Name=dyn_skirtBack_01_l, bone2Name=dyn_skirt_mstr, swing1LimitDegrees=12, swing2LimitDegrees=12, twistLimitDegrees=8}
 {bone1Name=dyn_skirtBack_01_r, bone2Name=dyn_skirt_mstr, swing1LimitDegrees=12, swing2LimitDegrees=12, twistLimitDegrees=8}
]
