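# Animation Blueprint for the Fish Princess cosmetic body, driving skirt secondary motion over the pose copied from the player mesh.
# A Control Rig feeds cached-pose RigidBody simulations blended for grind rail, jump/fall, crouch and default, bypassed in vehicle seats.
# Bone constraints and offsets then fix the skirt for emotes, sprint, ballet and seating.
# The event graph converts player states to alphas and resets dynamics after emotes and crouching.
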AnimBlueprint /BRCosmetics/Characters/Player/Female/Medium/Bodies/F_MED_Teriyaki_Fish_Princess/Meshes/F_MED_Teriyaki_Fish_Princess_AnimBP
target_skeleton=&/F_MED_Teriyaki_Fish_Princess_Skeleton
graphs [
 AnimGraph {nodes [
   n1 = "Constraint - dyn_skirtFront_02_r (PP)"!(ComponentPose, Alpha=1, ConstraintWeights_0=0.65, ConstraintWeights_1:float=0.35) -> (Pose) {
    alphaBoolBlend {blendOption=Linear}
    boneToModify {boneName=dyn_skirtFront_02_r}
    constraintSetup [{targetBone {boneName=thigh_r}}, {targetBone {boneName=pelvis}}]
    constraintWeights [1,1]
   }
   "Constraint - dyn_skirtFront_02_l (PP)"!(ComponentPose=@n1, Alpha=1, ConstraintWeights_0=0.85, ConstraintWeights_1:float=0.15) -> (Pose) {
    alphaBoolBlend {blendOption=Linear}
    boneToModify {boneName=dyn_skirtFront_02_l}
    constraintSetup [{targetBone {boneName=thigh_l}}, {targetBone {boneName=pelvis}}]
    constraintWeights [1,1]
   }
   "Constraint - dyn_skirtFront_03_l (PP)"!(ComponentPose, bAlphaBoolEnabled:bool=Get bIsSprinting(), ConstraintWeights_0=0.5, ConstraintWeights_1:float=0.5) -> (Pose) {
    alphaInputType=Bool
    alphaBoolBlend {blendInTime=0.5, blendOption=Linear, blendOutTime=0.1}
    boneToModify {boneName=dyn_skirtFront_03_l}
    constraintSetup [{targetBone {boneName=thigh_l}}, {targetBone {boneName=calf_l}}]
    constraintWeights [1,1]
   }
   Control Rig!(Source, Alpha=0.4, crouching:double=0) -> (Pose) {
    alphaBoolBlend {blendOption=Linear}
    controlRigAssetReference {blueprintRigClass=/BRCosmetics/Characters/Player/Female/Medium/Bodies/F_MED_Teriyaki_Fish_Princess/Meshes/CtrlRig/F_MED_TeriyakiPrincess_PostCtrlRig.F_MED_TeriyakiPrincess_PostCtrlRig_C}
   }
   Local To Component()
   "Transform (Modify) Bone - Bone: dyn_skirtSideBack_03_r"(Alpha=1, Translation="0,0,0", Rotation="0,-10,0", Scale="1,1,1") {
    rotationMode=BMM_Additive
    rotationSpace=BCS_BoneSpace
    alphaBoolBlend {blendOption=Linear}
    boneToModify {boneName=dyn_skirtSideBack_03_r}
    rotation {yaw=-10}
   }
   RigidBody() {
    bClampLinearTranslationLimitToRefPose=true
    bOverrideWorldGravity=true
    overridePhysicsAsset=&/F_MED_Teriyaki_Princess_GrindRail_Physics
    simulationSpace=BaseBoneSpace
    alphaBoolBlend {blendOption=Linear}
    alphaScaleBiasClamp {bInterpResult=true, interpSpeedDecreasing=5, interpSpeedIncreasing=5}
    baseBoneRef {boneName=root}
    componentAppliedLinearAccClamp=<2000 2000 2000>
    componentLinearAccScale=<0.3 0.3 0.55>
    componentLinearVelScale=<0.15 0.15 0.15>
    overrideWorldGravity {z=-980}
   }
   RigidBody!(
    ComponentPose
    Alpha=0
    OverrideWorldGravity:Vector="0,0,-980"
    ExternalForce
    SimSpaceSettings:Sim Space Settings Structure=(WorldAlpha=0,VelocityScaleZ=1,DampingAlpha=1,MaxLinearVelocity=10000,MaxAngularVelocity=10000,MaxLinearAcceleration=10000,MaxAngularAcceleration=10000,ExternalLinearDragV=(X=0,Y=0,Z=0),ExternalLinearVelocity=(X=0,Y=0,Z=0),ExternalAngularVelocity=(X=0,Y=0,Z=0))
   ) -> (Pose) {
    bClampLinearTranslationLimitToRefPose=true
    bOverrideWorldGravity=true
    overridePhysicsAsset=&/F_MED_Teriyaki_Fish_Princess_Physics
    simulationSpace=BaseBoneSpace
    alphaBoolBlend {blendOption=Linear}
    alphaScaleBiasClamp {bInterpResult=true, interpSpeedDecreasing=5, interpSpeedIncreasing=5}
    baseBoneRef {boneName=pelvis}
    componentAppliedLinearAccClamp=<2000 2000 2000>
    overrideWorldGravity {z=-980}
   }
   n2 = Get ballet() -> (ballet:double)
   n3 = Use cached pose 'postSimCache'()
   n4 = Use cached pose 'preSimulate'()
   n5 = Get vehicleBool() -> (vehicleBool:bool)
   n6 = Get bIsPlayingEmote()
   n7 = "Constraint - dyn_skirtFront_02_r (P)"!(
    ComponentPose=Local To Component(LocalPose=Blend Poses by bool(BlendPose_0=@n4, BlendPose_1=@n3, BlendTime_0=0.15, BlendTime_1=0.4, bActiveValue=@n5) {blendType=Linear})
    bAlphaBoolEnabled:bool=@n6
    ConstraintWeights_0=1
   ) -> (Pose) {
    alphaInputType=Bool
    alphaBoolBlend {blendOption=Linear}
    boneToModify {boneName=dyn_skirtFront_02_r}
    constraintSetup [{targetBone {boneName=thigh_r}}]
    constraintWeights [1]
   }
   n8 = "Constraint - dyn_skirtFront_02_l (P)"!(ComponentPose=@n7, bAlphaBoolEnabled:bool=@n6, ConstraintWeights_0=1) -> (Pose) {
    alphaInputType=Bool
    alphaBoolBlend {blendOption=Linear}
    boneToModify {boneName=dyn_skirtFront_02_l}
    constraintSetup [{targetBone {boneName=thigh_l}}]
    constraintWeights [1]
   }
   n9 = Get sprintAlpha() -> (sprintAlpha:double)
   n10 = "Constraint - dyn_skirtFront_02_r (PP)"(ComponentPose=@n8, Alpha=@n9, ConstraintWeights_0=1, ConstraintWeights_1=0) {
    alphaBoolBlend {blendInTime=0.5, blendOption=Linear, blendOutTime=0.1}
    alphaScaleBiasClamp {bClampResult=true, bInterpResult=true, clampMax=0.8, interpSpeedIncreasing=7}
    boneToModify {boneName=dyn_skirtFront_02_r}
    constraintSetup [{targetBone {boneName=thigh_r}}, {targetBone {boneName=pelvis}}]
    constraintWeights [1,1]
   }
   n11 = "Constraint - dyn_skirtFront_02_l (PP)"(ComponentPose=@n10, Alpha=@n9, ConstraintWeights_0=1, ConstraintWeights_1=0) {
    alphaBoolBlend {blendInTime=0.5, blendOption=Linear, blendOutTime=0.1}
    alphaScaleBiasClamp {bClampResult=true, bInterpResult=true, clampMax=0.8, interpSpeedIncreasing=7}
    boneToModify {boneName=dyn_skirtFront_02_l}
    constraintSetup [{targetBone {boneName=thigh_l}}, {targetBone {boneName=pelvis}}]
    constraintWeights [1,1]
   }
   n12 = "Transform (Modify) Bone - Bone: dyn_skirtFront_03_l"(ComponentPose=@n11, Alpha=@n9, Translation="0,0,10", Rotation="0,0,0", Scale="1,1,1") {
    translationMode=BMM_Additive
    translationSpace=BCS_BoneSpace
    alphaBoolBlend {blendInTime=0.3, blendOption=Linear, blendOutTime=0.05}
    alphaScaleBiasClamp {bInterpResult=true, interpSpeedIncreasing=7}
    boneToModify {boneName=dyn_skirtFront_03_l}
    translation {z=10}
   }
   n13 = "Constraint - dyn_skirtSideFront_02_r (PP)"!(ComponentPose=@n12, Alpha=@n2, ConstraintWeights_0=1, ConstraintWeights_1:float=0) -> (Pose) {
    alphaBoolBlend {blendOption=Linear}
    boneToModify {boneName=dyn_skirtSideFront_02_r}
    constraintSetup [{targetBone {boneName=dyn_skirtFront_02_r}}, {targetBone {boneName=dyn_skirtBack_02_r}}]
    constraintWeights [1,1]
   }
   n14 = "Constraint - dyn_skirtSideFront_02_l (PP)"!(ComponentPose=@n13, Alpha=@n2, ConstraintWeights_0=1, ConstraintWeights_1:float=0) -> (Pose) {
    alphaBoolBlend {blendOption=Linear}
    boneToModify {boneName=dyn_skirtSideFront_02_l}
    constraintSetup [{targetBone {boneName=dyn_skirtSideBack_02_l}}, {targetBone {boneName=dyn_skirtFront_02_l}}]
    constraintWeights [1,1]
   }
   n15 = "Constraint - dyn_skirtSideBack_02_r (PP)"!(ComponentPose=@n14, Alpha=@n2, ConstraintWeights_0=1, ConstraintWeights_1:float=0) -> (Pose) {
    alphaBoolBlend {blendOption=Linear}
    boneToModify {boneName=dyn_skirtSideBack_02_r}
    constraintSetup [{targetBone {boneName=dyn_skirtFront_02_r}}, {targetBone {boneName=dyn_skirtBack_02_r}}]
    constraintWeights [1,1]
   }
   n16 = "Constraint - dyn_skirtSideFront_03_r (PP)"!(ComponentPose=@n15, Alpha=@n2, ConstraintWeights_0=1, ConstraintWeights_1:float=0) -> (Pose) {
    alphaBoolBlend {blendOption=Linear}
    boneToModify {boneName=dyn_skirtSideFront_03_r}
    constraintSetup [{targetBone {boneName=calf_r}}, {targetBone {boneName=dyn_skirtBack_02_r}}]
    constraintWeights [1,1]
   }
   n17 = "Transform (Modify) Bone - Bone: dyn_skirtSideBack_02_r"(ComponentPose=@n16, Alpha=@n2, Translation="0,0,0", Rotation="0,-40,0", Scale="1,1,1") {
    rotationMode=BMM_Additive
    rotationSpace=BCS_BoneSpace
    alphaBoolBlend {blendOption=Linear}
    alphaScaleBiasClamp {bInterpResult=true, interpSpeedIncreasing=2}
    boneToModify {boneName=dyn_skirtSideBack_02_r}
    rotation {yaw=-40}
   }
   n18 = "Constraint - dyn_skirtSideBack_02_l (PP)"!(ComponentPose=@n17, Alpha=@n2, ConstraintWeights_0=1, ConstraintWeights_1:float=0) -> (Pose) {
    alphaBoolBlend {blendOption=Linear}
    boneToModify {boneName=dyn_skirtSideBack_02_l}
    constraintSetup [{targetBone {boneName=dyn_skirtSideBack_02_l}}, {targetBone {boneName=dyn_skirtFront_02_l}}]
    constraintWeights [1,1]
   }
   n19 = Get vehicleSeat() -> (vehicleSeat:double)
   n20 = "Constraint - dyn_skirtFront_01_r (PP)"!(ComponentPose=@n18, Alpha=@n19, ConstraintWeights_0=1, ConstraintWeights_1:float=0) -> (Pose) {
    alphaBoolBlend {blendOption=Linear}
    boneToModify {boneName=dyn_skirtFront_01_r}
    constraintSetup [{targetBone {boneName=thigh_r}}, {targetBone {boneName=dyn_skirtBack_02_r}}]
    constraintWeights [1,1]
   }
   n21 = "Constraint - dyn_skirtSideFront_01_r (PP)"!(ComponentPose=@n20, Alpha=@n19, ConstraintWeights_0=1, ConstraintWeights_1:float=0) -> (Pose) {
    alphaBoolBlend {blendOption=Linear}
    boneToModify {boneName=dyn_skirtSideFront_01_r}
    constraintSetup [{targetBone {boneName=thigh_r}}, {targetBone {boneName=dyn_skirtFront_02_l}}]
    constraintWeights [1,1]
   }
   n22 = "Constraint - dyn_skirtSideBack_01_r (PP)"!(ComponentPose=@n21, Alpha=@n19, ConstraintWeights_0=1, ConstraintWeights_1:float=0) -> (Pose) {
    alphaBoolBlend {blendOption=Linear}
    boneToModify {boneName=dyn_skirtSideBack_01_r}
    constraintSetup [{targetBone {boneName=thigh_r}}, {targetBone {boneName=dyn_skirtBack_02_r}}]
    constraintWeights [1,1]
   }
   n23 = "Constraint - dyn_skirtBack_01_r (PP)"!(ComponentPose=@n22, Alpha=@n19, ConstraintWeights_0=1, ConstraintWeights_1:float=0) -> (Pose) {
    alphaBoolBlend {blendOption=Linear}
    boneToModify {boneName=dyn_skirtBack_01_r}
    constraintSetup [{targetBone {boneName=thigh_r}}, {targetBone {boneName=dyn_skirtFront_02_l}}]
    constraintWeights [1,1]
   }
   n24 = "Constraint - dyn_skirtFront_03_r (PP)"!(ComponentPose=@n23, Alpha=@n19, ConstraintWeights_0=1, ConstraintWeights_1:float=0) -> (Pose) {
    alphaBoolBlend {blendOption=Linear}
    boneToModify {boneName=dyn_skirtFront_03_r}
    constraintSetup [{targetBone {boneName=calf_r}}, {targetBone {boneName=dyn_skirtBack_02_r}}]
    constraintWeights [1,1]
   }
   n25 = "Constraint - dyn_skirtSideFront_03_r (PP)"(ComponentPose=@n24, Alpha=@n19, ConstraintWeights_0=1, ConstraintWeights_1=0) {
    alphaBoolBlend {blendOption=Linear}
    boneToModify {boneName=dyn_skirtSideFront_03_r}
    constraintSetup [{targetBone {boneName=calf_r}}, {targetBone {boneName=dyn_skirtFront_02_l}}]
    constraintWeights [1,1]
   }
   n26 = "Constraint - dyn_skirtSideBack_03_r (PP)"!(ComponentPose=@n25, Alpha=@n19, ConstraintWeights_0=1, ConstraintWeights_1:float=0) -> (Pose) {
    alphaBoolBlend {blendOption=Linear}
    boneToModify {boneName=dyn_skirtSideBack_03_r}
    constraintSetup [{targetBone {boneName=calf_r}}, {targetBone {boneName=dyn_skirtBack_02_r}}]
    constraintWeights [1,1]
   }
   n27 = "Constraint - dyn_skirtBack_03_r (PP)"!(ComponentPose=@n26, Alpha=@n19, ConstraintWeights_0=1, ConstraintWeights_1:float=0) -> (Pose) {
    alphaBoolBlend {blendOption=Linear}
    boneToModify {boneName=dyn_skirtBack_03_r}
    constraintSetup [{targetBone {boneName=calf_r}}, {targetBone {boneName=dyn_skirtFront_02_l}}]
    constraintWeights [1,1]
   }
   n28 = "Constraint - dyn_skirtFront_01_l (PP)"!(ComponentPose=@n27, Alpha=@n19, ConstraintWeights_0=1, ConstraintWeights_1:float=0) -> (Pose) {
    alphaBoolBlend {blendOption=Linear}
    boneToModify {boneName=dyn_skirtFront_01_l}
    constraintSetup [{targetBone {boneName=thigh_l}}, {targetBone {boneName=dyn_skirtBack_02_r}}]
    constraintWeights [1,1]
   }
   n29 = "Constraint - dyn_skirtSideFront_01_l (PP)"!(ComponentPose=@n28, Alpha=@n19, ConstraintWeights_0=1, ConstraintWeights_1:float=0) -> (Pose) {
    alphaBoolBlend {blendOption=Linear}
    boneToModify {boneName=dyn_skirtSideFront_01_l}
    constraintSetup [{targetBone {boneName=thigh_l}}, {targetBone {boneName=dyn_skirtFront_02_l}}]
    constraintWeights [1,1]
   }
   n30 = "Constraint - dyn_skirtSideBack_01_l (PP)"!(ComponentPose=@n29, Alpha=@n19, ConstraintWeights_0=1, ConstraintWeights_1:float=0) -> (Pose) {
    alphaBoolBlend {blendOption=Linear}
    boneToModify {boneName=dyn_skirtSideBack_01_l}
    constraintSetup [{targetBone {boneName=thigh_l}}, {targetBone {boneName=dyn_skirtBack_02_r}}]
    constraintWeights [1,1]
   }
   n31 = "Constraint - dyn_skirtBack_01_l (PP)"!(ComponentPose=@n30, Alpha=@n19, ConstraintWeights_0=1, ConstraintWeights_1:float=0) -> (Pose) {
    alphaBoolBlend {blendOption=Linear}
    boneToModify {boneName=dyn_skirtBack_01_l}
    constraintSetup [{targetBone {boneName=thigh_l}}, {targetBone {boneName=dyn_skirtFront_02_l}}]
    constraintWeights [1,1]
   }
   n32 = "Constraint - dyn_skirtFront_03_l (PP)"!(ComponentPose=@n31, Alpha:float=@n19, ConstraintWeights_0=1, ConstraintWeights_1=0) -> (Pose) {
    alphaBoolBlend {blendOption=Linear}
    boneToModify {boneName=dyn_skirtFront_03_l}
    constraintSetup [{targetBone {boneName=calf_l}}, {targetBone {boneName=dyn_skirtBack_02_r}}]
    constraintWeights [1,1]
   }
   n33 = "Constraint - dyn_skirtSideFront_03_l (PP)"!(ComponentPose=@n32, Alpha=@n19, ConstraintWeights_0=1, ConstraintWeights_1:float=0) -> (Pose) {
    alphaBoolBlend {blendOption=Linear}
    boneToModify {boneName=dyn_skirtSideFront_03_l}
    constraintSetup [{targetBone {boneName=calf_l}}, {targetBone {boneName=dyn_skirtFront_02_l}}]
    constraintWeights [1,1]
   }
   n34 = "Constraint - dyn_skirtSideBack_03_l (PP)"!(ComponentPose=@n33, Alpha=@n19, ConstraintWeights_0=1, ConstraintWeights_1:float=0) -> (Pose) {
    alphaBoolBlend {blendOption=Linear}
    boneToModify {boneName=dyn_skirtSideBack_03_l}
    constraintSetup [{targetBone {boneName=calf_l}}, {targetBone {boneName=dyn_skirtBack_02_r}}]
    constraintWeights [1,1]
   }
   n35 = "Constraint - dyn_skirtBack_03_l (PP)"!(ComponentPose=@n34, Alpha=@n19, ConstraintWeights_0=1, ConstraintWeights_1:float=0) -> (Pose) {
    alphaBoolBlend {blendOption=Linear}
    boneToModify {boneName=dyn_skirtBack_03_l}
    constraintSetup [{targetBone {boneName=calf_l}}, {targetBone {boneName=dyn_skirtFront_02_l}}]
    constraintWeights [1,1]
   }
   Output Pose(Result=Component To Local(ComponentPose=@n35)) {layerGroup=None}
   n36 = Control Rig!(
    Source=Copy Pose From Mesh() {bUseAttachedParent=true}
    Alpha
    crouching=Get crouchAlpha() -> (crouchAlpha:double)
    jumping:double=Get jumpAlpha() -> (jumpAlpha:double)
    sprinting:double=Get sprintAlpha()
    isDancing:double=Get isDancing() -> (isDancing:double)
    is_sliding_alpha:double=Get is_sliding_alpha() -> (is_sliding_alpha:double)
    is_on_grind_rail_alpha:double=Get is_on_grind_rail_alpha() -> (is_on_grind_rail_alpha:double)
   ) -> (Pose) {
    alphaBoolBlend {blendOption=Linear}
    controlRigAssetReference {blueprintRigClass=/BRCosmetics/Characters/Player/Female/Medium/Bodies/F_MED_Teriyaki_Fish_Princess/Meshes/CtrlRig/F_MED_Teriyaki_Fish_Princess_CtrlRig.F_MED_Teriyaki_Fish_Princess_CtrlRig_C}
   }
   Save cached pose 'preSimulate'(Pose=@n36) {cacheName=preSimulate}
   n37 = RigidBody!(ComponentPose=Local To Component(LocalPose=Use cached pose 'preSimulate'()), Alpha=Get crouchAlpha(), ExternalForce) -> (Pose) {
    bClampLinearTranslationLimitToRefPose=true
    bOverrideWorldGravity=true
    overridePhysicsAsset=&F_MED_Teriyaki_Princess_GrindRail_Physics
    simulationSpace=BaseBoneSpace
    alphaBoolBlend {blendOption=Linear}
    alphaScaleBiasClamp {bInterpResult=true, interpSpeedDecreasing=5, interpSpeedIncreasing=5}
    baseBoneRef {boneName=root}
    componentAppliedLinearAccClamp=<2000 2000 2000>
    overrideWorldGravity {z=-200}
    simSpaceSettings {velocityScaleZ=0.5}
   }
   n38 = RigidBody!(
    ComponentPose=Local To Component(LocalPose=Use cached pose 'preSimulate'())
    Alpha=Get crouchAlphaINV() -> (crouchAlphaINV:double)
    OverrideWorldGravity:Vector="0,0,-980"
    ExternalForce
    SimSpaceSettings:Sim Space Settings Structure=(WorldAlpha=0,VelocityScaleZ=1,DampingAlpha=1,MaxLinearVelocity=10000,MaxAngularVelocity=10000,MaxLinearAcceleration=10000,MaxAngularAcceleration=10000,ExternalLinearDragV=(X=0,Y=0,Z=0),ExternalLinearVelocity=(X=0,Y=0,Z=0),ExternalAngularVelocity=(X=0,Y=0,Z=0))
   ) -> (Pose) {
    bClampLinearTranslationLimitToRefPose=true
    bOverrideWorldGravity=true
    overridePhysicsAsset=&F_MED_Teriyaki_Fish_Princess_Physics
    simulationSpace=BaseBoneSpace
    alphaBoolBlend {blendOption=Linear}
    alphaScaleBiasClamp {bInterpResult=true, interpSpeedDecreasing=5, interpSpeedIncreasing=5}
    baseBoneRef {boneName=pelvis}
    componentAppliedLinearAccClamp=<2000 2000 2000>
    componentLinearAccScale=<0.2 0.2 0.2>
    componentLinearVelScale=<0.15 0.15 0.15>
    overrideWorldGravity {z=-980}
   }
   n39 = Blend Poses by bool(
    BlendPose_0=Component To Local(ComponentPose=@n37)
    BlendPose_1=Component To Local(ComponentPose=@n38)
    BlendTime_0=0.25
    BlendTime_1=0.25
    bActiveValue=Get bIsCrouching()
   ) {blendType=Linear}
   n40 = Get jumpAlpha()
   n41 = "Transform (Modify) Bone - Bone: dyn_skirtSideFront_01_r"(
    ComponentPose=Local To Component(LocalPose=Use cached pose 'preSimulate'())
    Alpha=@n40
    Translation="0,0,0"
    Rotation="-10,-15,0"
    Scale="1,1,1"
   ) {
    rotationMode=BMM_Additive
    rotationSpace=BCS_BoneSpace
    alphaBoolBlend {blendOption=Linear}
    alphaScaleBiasClamp {bInterpResult=true, interpSpeedDecreasing=5, interpSpeedIncreasing=5}
    boneToModify {boneName=dyn_skirtSideFront_01_r}
    rotation {pitch=-10, yaw=-15}
   }
   n42 = "Transform (Modify) Bone - Bone: dyn_skirtSideFront_02_r"(ComponentPose=@n41, Alpha=@n40, Translation="0,0,0", Rotation="-5,-15,0", Scale="1,1,1") {
    rotationMode=BMM_Additive
    rotationSpace=BCS_BoneSpace
    alphaBoolBlend {blendOption=Linear}
    alphaScaleBiasClamp {bInterpResult=true, interpSpeedDecreasing=5, interpSpeedIncreasing=5}
    boneToModify {boneName=dyn_skirtSideFront_02_r}
    rotation {pitch=-5, yaw=-15}
   }
   n43 = "Transform (Modify) Bone - Bone: dyn_skirtSideBack_01_r"(ComponentPose=@n42, Alpha=@n40, Translation="0,0,0", Rotation="-10,-10,0", Scale="1,1,1") {
    rotationMode=BMM_Additive
    rotationSpace=BCS_BoneSpace
    alphaBoolBlend {blendOption=Linear}
    alphaScaleBiasClamp {bInterpResult=true, interpSpeedDecreasing=5, interpSpeedIncreasing=5}
    boneToModify {boneName=dyn_skirtSideBack_01_r}
    rotation {pitch=-10, yaw=-10}
   }
   n44 = "Transform (Modify) Bone - Bone: dyn_skirtSideBack_02_l"(ComponentPose=@n43, Alpha=@n40, Translation="0,0,0", Rotation="20,0,0", Scale="1,1,1") {
    rotationMode=BMM_Additive
    rotationSpace=BCS_BoneSpace
    alphaBoolBlend {blendOption=Linear}
    boneToModify {boneName=dyn_skirtSideBack_02_l}
    rotation {pitch=20}
   }
   n45 = "Transform (Modify) Bone - Bone: dyn_skirtSideBack_01_l"(ComponentPose=@n44, Alpha=@n40, Translation="0,0,0", Rotation="0,-20,0", Scale="1,1,1") {
    rotationMode=BMM_Additive
    rotationSpace=BCS_BoneSpace
    alphaBoolBlend {blendOption=Linear}
    boneToModify {boneName=dyn_skirtSideBack_01_l}
    rotation {yaw=-20}
   }
   n46 = RigidBody!(ComponentPose=@n45, Alpha=@n40, ExternalForce) -> (Pose) {
    bClampLinearTranslationLimitToRefPose=true
    bOverrideWorldGravity=true
    overridePhysicsAsset=&/F_MED_TeriyakiFish_Princess_JumpPhysics
    simulationSpace=BaseBoneSpace
    alphaBoolBlend {blendOption=Linear}
    alphaScaleBiasClamp {bInterpResult=true, interpSpeedDecreasing=4, interpSpeedIncreasing=4}
    baseBoneRef {boneName=root}
    componentAppliedLinearAccClamp=<2000 2000 2000>
    componentLinearAccScale=<0.3 0.3 0.55>
    componentLinearVelScale=<0.15 0.15 0.15>
    overrideWorldGravity {z=-980}
   }
   n47 = Component To Local(ComponentPose=@n46)
   n48 = Get jumpOrFall() -> (jumpOrFall:bool)
   n49 = RigidBody(ComponentPose=Local To Component(LocalPose=Use cached pose 'preSimulate'())) {
    bClampLinearTranslationLimitToRefPose=true
    bOverrideWorldGravity=true
    overridePhysicsAsset=&F_MED_Teriyaki_Princess_GrindRail_Physics
    simulationSpace=BaseBoneSpace
    alphaBoolBlend {blendOption=Linear}
    alphaScaleBiasClamp {bInterpResult=true, interpSpeedDecreasing=5, interpSpeedIncreasing=5}
    baseBoneRef {boneName=root}
    componentAppliedLinearAccClamp=<2000 2000 2000>
    componentLinearAccScale=<0.3 0.3 0.55>
    componentLinearVelScale=<0.15 0.15 0.15>
    overrideWorldGravity {z=-980}
   }
   Save cached pose 'postSimCache'(Pose=Blend Poses by bool(
     BlendPose_0=Component To Local(ComponentPose=@n49)
     BlendPose_1=Blend Poses by bool(BlendPose_0=@n47, BlendPose_1=@n39, BlendTime_0=0.15, BlendTime_1=0.4, bActiveValue=@n48) {blendType=Linear}
     BlendTime_0=0.15
     BlendTime_1=0.4
     bActiveValue=Get bIsOnGrindRail()
    ) {blendType=Linear}) {cacheName=postSimCache}
   Get bIsJumping()
  ]}
 EventGraph {nodes [
   Event BlueprintUpdateAnimation()
   > n1 = Sequence!(execute) -> (then_0, then_1, then_2:Exec)
   n2 = NOT Boolean(A=Get bIsCrouchSprinting())
   n3 = Get bIsSprinting()
   > Set sprintAlpha(sprintAlpha:double="To Float (Boolean)"(InBool=AND Boolean(A=@n3, B=@n2))) -> (Output_Get:double)
   n4 = AND Boolean!(A=NOT Boolean(A=Get bIsMeatballDriver() -> (bIsMeatballDriver:bool)), B=true, C:bool=Get bIsInVehicleSeat() -> (bIsInVehicleSeat:bool)) -> (ReturnValue)
   > Set vehicleSeat(vehicleSeat:double="To Float (Boolean)"(InBool=@n4)) -> (Output_Get:double)
   > Set vehicleBool(vehicleBool:bool=@n4) -> (Output_Get:bool)
   n5 = GetOwnerFortPlayerAnimInstance!(self:Custom Character Part Anim Instance Object Reference) -> (ReturnValue:Fort Player Anim Instance Object Reference)
   > Is Valid(InputObject=@n5)
   n6 = Montage_IsPlaying(self=@n5, Montage=/BRCosmetics/Animation/Game/MainPlayer/Emotes/BalletSpin/Emote_BalletSpin_CMM_M)
   n7 = Montage_IsPlaying(self=@n5, Montage=/BRCosmetics/Animation/Game/MainPlayer/Emotes/BalletJumps/Emote_BalletJumps_CMM_M)
   > Set ballet(ballet:double="To Float (Boolean)"(InBool=OR Boolean(A=@n7, B=@n6))) -> (Output_Get:double)
   Set is_sliding_alpha(execute=@n1.then_1, is_sliding_alpha:double="To Float (Boolean)"(InBool=Get bIsSliding())) -> (Output_Get:double)
   > Set is_on_grind_rail_alpha(is_on_grind_rail_alpha:double="To Float (Boolean)"(InBool=Get bIsOnGrindRail())) -> (Output_Get:double)
   n8 = Gravity Override State Function DEPRECATED(
    execute=@n1.then_2
    CharacterPartAnimInstance=Self-Reference()
    GameTimeInSeconds=GetGameTimeInSeconds()
    GravityOverrideParams=Get Gravity Override Params()
    RBAN Params=Get RBAN Params()
   )
   > Set Gravity Override(Gravity Override=@n8)
   > Set RBAN Settings Out(RBAN Settings Out:RBANSetting Struct Structure=@n8.RBAN SettingsOut) -> (Output_Get:RBANSetting Struct Structure)
   Event OnBeginPlayingEmote() {is=title K2Node_Event}
   > Set isDancing(isDancing:double="To Float (Boolean)"(InBool=Get bIsPlayingEmote())) -> (Output_Get:double)
   n9 = GetOwnerFortPlayerAnimInstance()
   n10 = Montage_IsPlaying(self=@n9, Montage=&Emote_BalletSpin_CMM_M)
   n11 = Montage_IsPlaying(self=@n9, Montage=&Emote_BalletJumps_CMM_M)
   > Set ballet(ballet="To Float (Boolean)"(InBool=OR Boolean(A=@n11, B=@n10)))
   Event OnEndFalling()
   > Set jumpAlpha(jumpAlpha:double=0) -> (Output_Get:double)
   > Set jumpOrFall(jumpOrFall:bool=false) -> (Output_Get:bool)
   Event OnEndPlayingEmote() {is=title K2Node_Event}
   > Set isDancing(isDancing="To Float (Boolean)"(InBool=Get bIsPlayingEmote()))
   > Set ballet(ballet=0)
   > Delay(Duration=0.085)
   > ResetDynamics(InTeleportType=TeleportPhysics)
   > ResetDynamics(InTeleportType=ResetPhysics)
   Event OnBeginCrouching()
   n12 = Get bIsCrouching()
   > Set crouchAlpha(crouchAlpha:double="To Float (Boolean)"(InBool=@n12)) -> (Output_Get:double)
   > Set crouchAlphaINV(crouchAlphaINV:double=float - float(B="To Float (Boolean)"(InBool=@n12)) {is=title K2Node_CallFunction}) -> (Output_Get:double)
   Event OnEndCrouching()
   n13 = Get bIsCrouching()
   > Set crouchAlpha(crouchAlpha="To Float (Boolean)"(InBool=@n13))
   > Set crouchAlphaINV(crouchAlphaINV=float - float(B="To Float (Boolean)"(InBool=@n13)))
   > Delay(Duration=0.085)
   > ResetDynamics(InTeleportType=TeleportPhysics)
   > ResetDynamics(InTeleportType=ResetPhysics)
   Event OnBeginJumping()
   n14 = OR Boolean(A=Get bIsJumping(), B=Get bIsFalling())
   > Set jumpAlpha(jumpAlpha="To Float (Boolean)"(InBool=@n14))
   > Set jumpOrFall(jumpOrFall=@n14)
   NOT Boolean(A=Get bIsAntelopeVehicleDriver() -> (bIsAntelopeVehicleDriver:bool))
  ]}
]
variables [
 ballet {type=number}
 crouchAlpha {type=number}
 crouchAlphaINV {type=number, default=1}
 Gravity Override {type=Vector}
 Gravity Override Params {
  type=GravityOverrideParamsStruct
  default {
   skydive_DivingDown {gravityOverride=<0 -700 0>}
   skydive_FloatingDown {jointName=pelvis, gravityOverride=<0 -500 0>, windAmplitude=<50 50 50>, windOffset=<0 400 0>}
   skydive_GoingUp {gravityOverride=<0 0 0>, windAmplitude=<100 100 100>}
   skydive_Parachuting {gravityOverride=<0 -800 0>}
  }
 }
 is_on_grind_rail_alpha {type=number}
 is_sliding_alpha {type=number}
 isDancing {type=number}
 jumpAlpha {type=number}
 jumpOrFall {type=boolean}
 RBAN Params {
  type=RBANParamsStruct
  default {
   dBNO {simSpaceSettings {
     maxAngularAcceleration=1000
     maxAngularVelocity=3
     maxLinearAcceleration=300
     maxLinearVelocity=700
     velocityScaleZ=0
     externalLinearDragV=<0 0.5 0>
    }}
   dBNO_Carried {simSpaceSettings {
     maxAngularAcceleration=1000
     maxAngularVelocity=3
     maxLinearAcceleration=300
     maxLinearVelocity=700
     velocityScaleZ=0
     externalLinearDragV=<0 0.5 0>
    }}
   emoteOrMelee {simSpaceSettings {
     maxAngularAcceleration=1000
     maxAngularVelocity=3
     maxLinearAcceleration=300
     maxLinearVelocity=700
     velocityScaleZ=0
     externalLinearDragV=<0 0.5 0>
    }}
   front_End {simSpaceSettings {
     maxAngularAcceleration=1000
     maxAngularVelocity=3
     maxLinearAcceleration=300
     maxLinearVelocity=700
     velocityScaleZ=0
     externalLinearDragV=<0 0.5 0>
    }}
   onGround_Crouch_Moving {simSpaceSettings {
     maxAngularAcceleration=1000
     maxAngularVelocity=3
     maxLinearAcceleration=300
     maxLinearVelocity=700
     velocityScaleZ=0
     externalLinearDragV=<0 0.5 0>
    }}
   onGround_Falling {simSpaceSettings {
     maxAngularAcceleration=1000
     maxAngularVelocity=3
     maxLinearAcceleration=300
     maxLinearVelocity=700
     velocityScaleZ=0
     externalLinearDragV=<0 0.5 0>
    }}
   onGround_Moving {simSpaceSettings {
     maxAngularAcceleration=1000
     maxAngularVelocity=3
     maxLinearAcceleration=300
     maxLinearVelocity=700
     velocityScaleZ=0
     externalLinearDragV=<0 0.5 0>
    }}
   onGround_Standing {simSpaceSettings {
     maxAngularAcceleration=1000
     maxAngularVelocity=3
     maxLinearAcceleration=300
     maxLinearVelocity=700
     velocityScaleZ=0
     externalLinearDragV=<0 0.5 0>
    }}
   rail_Grinding {simSpaceSettings {
     maxAngularAcceleration=1000
     maxAngularVelocity=3
     maxLinearAcceleration=300
     maxLinearVelocity=700
     velocityScaleZ=0
     worldAlpha=0.1
     externalLinearDragV=<0 0.5 0>
    }}
   riding_Moving {simSpaceSettings {
     maxAngularAcceleration=1000
     maxAngularVelocity=3
     maxLinearAcceleration=300
     maxLinearVelocity=700
     velocityScaleZ=0
     externalLinearDragV=<0 0.5 0>
    }}
   riding_Moving_Targeting {simSpaceSettings {
     maxAngularAcceleration=1000
     maxAngularVelocity=3
     maxLinearAcceleration=300
     maxLinearVelocity=700
     velocityScaleZ=0
     externalLinearDragV=<0 0.5 0>
    }}
   riding_Standing {simSpaceSettings {
     maxAngularAcceleration=1000
     maxAngularVelocity=3
     maxLinearAcceleration=300
     maxLinearVelocity=700
     velocityScaleZ=0
     externalLinearDragV=<0 0.5 0>
    }}
   skydive_DivingDown {simSpaceSettings {
     maxAngularAcceleration=1000
     maxAngularVelocity=3
     maxLinearAcceleration=300
     maxLinearVelocity=700
     velocityScaleZ=0
     externalLinearDragV=<0 0.5 0>
    }}
   skydive_FloatingDown {
    componentLinearAccScale=<0.1 0.1 0.1>
    componentLinearVelScale=<0.1 0.1 0.1>
    simSpaceSettings {
     maxAngularAcceleration=0
     maxAngularVelocity=0
     maxLinearAcceleration=0
     maxLinearVelocity=0
     velocityScaleZ=0
     worldAlpha=0
     externalLinearDragV=<0 0 0>
    }
   }
   skydive_GoingUp {
    componentLinearAccScale=<0.1 0.1 0.1>
    componentLinearVelScale=<0.1 0.1 0.1>
    simSpaceSettings {
     maxAngularAcceleration=0
     maxAngularVelocity=0
     maxLinearAcceleration=0
     maxLinearVelocity=0
     velocityScaleZ=0
     worldAlpha=0
     externalLinearDragV=<0 0 0>
    }
   }
   skydive_Parachuting {
    componentLinearAccScale=<0 0.1 0.3>
    componentLinearVelScale=<0 0.1 0.3>
    simSpaceSettings {
     maxAngularAcceleration=1000
     maxAngularVelocity=3
     maxLinearAcceleration=300
     maxLinearVelocity=700
     velocityScaleZ=0
     externalLinearDragV=<0 0.5 0>
    }
   }
   skydive_Surfing {simSpaceSettings {
     maxAngularAcceleration=1000
     maxAngularVelocity=3
     maxLinearAcceleration=300
     maxLinearVelocity=700
     velocityScaleZ=0
     externalLinearDragV=<0 0.5 0>
    }}
   swimming {simSpaceSettings {
     maxAngularAcceleration=1000
     maxAngularVelocity=3
     maxLinearAcceleration=300
     maxLinearVelocity=700
     velocityScaleZ=0
     externalLinearDragV=<0 0.5 0>
    }}
   tactical_Sprint {simSpaceSettings {
     maxAngularAcceleration=1000
     maxAngularVelocity=3
     maxLinearAcceleration=300
     maxLinearVelocity=700
     velocityScaleZ=0
     externalLinearDragV=<0 0.5 0>
    }}
   ziplining {simSpaceSettings {
     maxAngularAcceleration=1000
     maxAngularVelocity=3
     maxLinearAcceleration=300
     maxLinearVelocity=700
     velocityScaleZ=0
     externalLinearDragV=<0 0.5 0>
    }}
  }
 }
 RBAN Settings Out {
  type=RBANSettingStruct
  default {
   componentLinearAccScale=<0 0 0>
   componentLinearVelScale=<0 0 0>
   simSpaceSettings {
    dampingAlpha=1
    maxAngularAcceleration=1000
    maxAngularVelocity=3
    maxLinearAcceleration=300
    maxLinearVelocity=700
    velocityScaleZ=0
    worldAlpha=1
    externalAngularVelocity=<0 0 0>
    externalLinearDragV=<0 0.5 0>
    externalLinearVelocity=<0 0 0>
   }
  }
 }
 sprintAlpha {type=number}
 Strafing {type=Vector}
 vehicleBool {type=boolean}
 vehicleSeat {type=number}
]
