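# Physics asset for the Teriyaki Fish Princess cosmetic body, covering the secondary motion of the character's hat dangles and hat cloth.
# Two box-body dangle chains and three capsule cloth chains (front, mid, back), each three bones long, hang from a two-capsule head body through angle-limited constraints.
# Extra constraints with free swing and twist tie the front and back cloth tips to the mid tip, linking the strips like a single sheet.
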PhysicsAsset /BRCosmetics/Characters/Player/Female/Medium/Bodies/F_MED_Teriyaki_Fish_Princess/Meshes/Parts/F_MED_Teriyaki_Fish_Princess_Physics
bodies {
 dyn_hatDangle_01_l [{shapeType=Box, extentX=1.3699, extentY=2.3501, extentZ=4.19, center=<2.044 -0.703 0>, rotation=<rot 0 90 -90>}]
 dyn_hatDangle_01_r [{shapeType=Box, extentX=1.3699, extentY=2.3501, extentZ=4.19, center=<-2.044 0.703 0>, rotation=<rot 0 90 90>}]
 dyn_hatDangle_02_l [{shapeType=Box, extentX=1.01, extentY=2.3298, extentZ=4.384, center=<2.0227 -0.0001 0>, rotation=<rot 0 90 -90>}]
 dyn_hatDangle_02_r [{shapeType=Box, extentX=1.01, extentY=2.3298, extentZ=4.384, center=<-2.0227 0.0001 0>, rotation=<rot 0 90 90>}]
 dyn_hatDangle_03_l [{shapeType=Box, extentX=1.2825, extentY=4.8308, extentZ=5.2195, center=<3.8358 0 0.0522>, rotation=<rot 0 90 -45>}]
 dyn_hatDangle_03_r [{shapeType=Box, extentX=1.2825, extentY=4.8308, extentZ=5.2195, center=<-3.8358 0 -0.0522>, rotation=<rot 0 90 135>}]
 dyn_hatclothBack_01 [{shapeType=Capsule, length=2.0111, radius=1.5985, center=<1.75 -0.8163 0>, rotation=<rot 90 0 0>}]
 dyn_hatclothBack_02 [{shapeType=Capsule, length=2.8346, radius=1.9797, center=<1.8459 -1.1352 0>, rotation=<rot 90 -167.0926 -171.922>}]
 dyn_hatclothBack_03 [{shapeType=Capsule, length=3.3485, radius=2.5883, center=<2.3592 0 0>, rotation=<rot 90 0 0>}]
 dyn_hatclothFront_01 [{shapeType=Capsule, length=1.646, radius=1.5215, center=<1.5 -0.5998 0>, rotation=<rot 90 0 0>}]
 dyn_hatclothFront_02 [{shapeType=Capsule, length=2.5004, radius=1.6711, center=<1.5 -0.5495 0>, rotation=<rot 90 0 0>}]
 dyn_hatclothFront_03 [{shapeType=Capsule, length=3.7409, radius=1.6454, center=<1.5 0.0887 0>, rotation=<rot 90 0 0>}]
 dyn_hatclothMid_01 [{shapeType=Capsule, length=3.6695, radius=1.5945, center=<1.75 -0.8953 0>, rotation=<rot 90 0 0>}]
 dyn_hatclothMid_02 [{shapeType=Capsule, length=4.5921, radius=2.3923, center=<1.75 -0.5495 0>, rotation=<rot 90 0 0>}]
 dyn_hatclothMid_03 [{shapeType=Capsule, length=5.5084, radius=2.6737, center=<1.75 0 0>, rotation=<rot 90 0 0>}]
 head [
  {shapeType=Capsule, length=4.155, radius=13.5221, center=<1.9355 5.2796 -0.0001>, rotation=<rot 0 -20.1331 89.9999>}
  {shapeType=Capsule, length=19.0744, radius=10.9256, center=<19.0052 2.5114 4.2889>, rotation=<rot -59.9237 -31.4388 27.7172>}
 ]
}
constraints [
 {bone1Name=dyn_hatDangle_02_r, bone2Name=dyn_hatDangle_01_r, swing1LimitDegrees=20, swing2LimitDegrees=20, twistLimitDegrees=8}
 {bone1Name=dyn_hatDangle_03_r, bone2Name=dyn_hatDangle_02_r, swing1LimitDegrees=20, swing2LimitDegrees=20, twistLimitDegrees=8}
 {bone1Name=dyn_hatDangle_02_l, bone2Name=dyn_hatDangle_01_l, swing1LimitDegrees=20, swing2LimitDegrees=20, twistLimitDegrees=8}
 {bone1Name=dyn_hatDangle_03_l, bone2Name=dyn_hatDangle_02_l, swing1LimitDegrees=20, swing2LimitDegrees=20, twistLimitDegrees=8}
 {bone1Name=dyn_hatDangle_01_r, bone2Name=head, swing1LimitDegrees=12, swing2LimitDegrees=12, twistLimitDegrees=12}
 {bone1Name=dyn_hatDangle_01_l, bone2Name=head, swing1LimitDegrees=12, swing2LimitDegrees=12, twistLimitDegrees=12}
 {bone1Name=dyn_hatclothFront_01, bone2Name=head, swing1LimitDegrees=12, swing2LimitDegrees=12, twistLimitDegrees=9}
 {bone1Name=dyn_hatclothFront_02, bone2Name=dyn_hatclothFront_01, swing1LimitDegrees=45, swing2LimitDegrees=45, twistLimitDegrees=45}
 {bone1Name=dyn_hatclothFront_03, bone2Name=dyn_hatclothFront_02, swing1LimitDegrees=12, swing2LimitDegrees=12, twistLimitDegrees=9}
 {bone1Name=dyn_hatclothMid_01, bone2Name=head, swing1LimitDegrees=12, swing2LimitDegrees=12, twistLimitDegrees=9}
 {bone1Name=dyn_hatclothMid_02, bone2Name=dyn_hatclothMid_01, swing1LimitDegrees=12, swing2LimitDegrees=12, twistLimitDegrees=9}
 {bone1Name=dyn_hatclothMid_03, bone2Name=dyn_hatclothMid_02, swing1LimitDegrees=12, swing2LimitDegrees=12, twistLimitDegrees=9}
 {bone1Name=dyn_hatclothBack_01, bone2Name=head, swing1LimitDegrees=14, swing2LimitDegrees=12, twistLimitDegrees=9}
 {bone1Name=dyn_hatclothBack_02, bone2Name=dyn_hatclothBack_01, swing1LimitDegrees=14, swing2LimitDegrees=12, twistLimitDegrees=9}
 {bone1Name=dyn_hatclothBack_03, bone2Name=dyn_hatclothBack_02, swing1LimitDegrees=12, swing2LimitDegrees=12, twistLimitDegrees=9}
 {
  bone1Name=dyn_hatclothFront_03
  bone2Name=dyn_hatclothMid_03
  swing1LimitDegrees=45
  swing1Motion=Free
  swing2LimitDegrees=45
  swing2Motion=Free
  twistLimitDegrees=45
  twistMotion=Free
 }
 {
  bone1Name=dyn_hatclothBack_03
  bone2Name=dyn_hatclothMid_03
  swing1LimitDegrees=45
  swing1Motion=Free
  swing2LimitDegrees=45
  swing2Motion=Free
  twistLimitDegrees=45
  twistMotion=Free
 }
]
